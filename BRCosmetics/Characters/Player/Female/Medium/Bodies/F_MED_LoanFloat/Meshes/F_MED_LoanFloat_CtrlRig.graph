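# Control Rig for the F_MED_LoanFloat female medium body cosmetic, adding corrective deformation and arm-driven coat adjustments.
# After the shared DeformRig function runs, spherical pose readers measure each upper arm's up, forward and down alphas against spine_05.
# Those alphas weight parent constraints and transform offsets on the coat master bones, and squared down alphas drive armpit correctives.
# Only the left down alpha is also written to a curve.
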ControlRigBlueprint /BRCosmetics/Characters/Player/Female/Medium/Bodies/F_MED_LoanFloat/Meshes/F_MED_LoanFloat_CtrlRig
graphs [
 ContainedGraph {nodes [
   n1 = RigVMFunctionEntryNode()
   SequenceAggregate = &RigVMFunction_Sequence(ExecuteContext=@n1.ExecuteContext)
   SequenceAggregate_1 = &RigVMFunction_Sequence(ExecuteContext=@SequenceAggregate.B)
   SequenceAggregate_2 = &RigVMFunction_Sequence(ExecuteContext=@SequenceAggregate_1.B)
   SequenceAggregate_3 = &RigVMFunction_Sequence(ExecuteContext=@SequenceAggregate_2.B)
   RigVMFunctionReturnNode(A=@SequenceAggregate.A, B=@SequenceAggregate_1.A, C=@SequenceAggregate_2.A, D=@SequenceAggregate_3.A, E=@SequenceAggregate_3.B)
  ]}
 RigVMFunctionLibrary {links [], nodes []}
 RigVMModel {nodes [
   RigUnit_BeginExecution = &RigUnit_BeginExecution()
   FN_DeformRig_FNC = RigVMFunctionReferenceNode(
    DeformRig_ Curve="(deformRig_l_upperarm_up_alpha_72_#1=False,deformRig_r_upperarm_up_alpha_73_#2=False,deformRig_l_upperarm_fr_alpha_74_#3=False,deformRig_r_upperarm_fr_alpha_75_CB51B8 ...(truncated)"
    DeformRig_Bool="(upperarm_1_#4=True,elbow_3_#5=True,wrist_7_#6=True,groin_8_#7=True,glute_17_#8=True,knee_9_40C408324CA261 ...(truncated)"
    DeformRig_Debug=(upperarm_debug_18_#4=False,wrist_debug_20_#6=False,lat_debug_26_#9=False,groinGlute_debug_27_#7=False)
    DeformRig_Value="(upperarm_up_18_#10=(Rotation=(X=0.0,Y=-0.0,Z=0.0,W=1.0),Translation=(X=0,Y=0,Z=3),Scale3D=(X=0.70,Y=1.0,Z=1.0)),upperarm_fwd_fr_19_#11=(Rotation=(X=0.0,Y=-0.0,Z=0.0,W=1.0),Translation=(X=-5,Y=10,Z= ...(truncated)"
    ExecuteContext=@RigUnit_BeginExecution.ExecutePin
   )
   > SequenceAggregate = RigVMAggregateNode()
   n1 = &RigUnit_SphericalPoseReader(
    ActiveRegionScaleFactors=(PositiveWidth=2,NegativeWidth=2,PositiveHeight=2,NegativeHeight=2)
    ActiveRegionSize=0.1
    Debug=(bDrawDebug=false,DebugScale=25,DebugSegments=20,DebugThickness=0.25)
    DriverAxis=(X=1,Y=0,Z=0)
    DriverItem=(Type=Bone,Name="upperarm_l")
    FalloffRegionScaleFactors=(PositiveWidth=0.5,NegativeWidth=0.5,PositiveHeight=0.5,NegativeHeight=0.5)
    FalloffSize=0.5
    OptionalParentItem=(Type=Bone,Name="spine_05")
    RotationOffset=(X=0,Y=10,Z=0)
    ExecutePin=@SequenceAggregate.A
   )
   > RigVMVariableNode(Value=@n1.OutputParam) {variable=spr_upperarm_l_up_alpha}
   > n2 = &RigUnit_SphericalPoseReader(
    ActiveRegionScaleFactors=(PositiveWidth=2,NegativeWidth=2,PositiveHeight=2,NegativeHeight=2)
    ActiveRegionSize=0.1
    Debug=(bDrawDebug=false,DebugScale=25,DebugSegments=20,DebugThickness=0.25)
    DriverAxis=(X=-1,Y=0,Z=0)
    DriverItem=(Type=Bone,Name="upperarm_r")
    FalloffRegionScaleFactors=(PositiveWidth=0.5,NegativeWidth=0.5,PositiveHeight=0.5,NegativeHeight=0.5)
    FalloffSize=0.5
    FlipHeightScaling=true
    FlipWidthScaling=false
    OptionalParentItem=(Type=Bone,Name="spine_05")
    RotationOffset=(X=180,Y=-10,Z=180)
   )
   > RigVMVariableNode(Value=@n2.OutputParam) {variable=spr_upperarm_r_up_alpha}
   > n3 = &RigUnit_SphericalPoseReader(
    ActiveRegionScaleFactors=(PositiveWidth=0.1,NegativeWidth=1,PositiveHeight=0.1,NegativeHeight=1)
    ActiveRegionSize=0.1
    Debug=(bDrawDebug=false,DebugScale=25,DebugSegments=20,DebugThickness=0.25)
    DriverAxis=(X=1,Y=0,Z=0)
    DriverItem=(Type=Bone,Name="upperarm_l")
    FalloffRegionScaleFactors=(PositiveWidth=0.8,NegativeWidth=1,PositiveHeight=0.8,NegativeHeight=1)
    FalloffSize=0.5
    OptionalParentItem=(Type=Bone,Name="spine_05")
    RotationOffset=(X=85,Y=10,Z=-10)
   )
   > RigVMVariableNode(Value=@n3.OutputParam) {variable=spr_upperarm_l_f_alpha}
   > n4 = &RigUnit_SphericalPoseReader(
    ActiveRegionScaleFactors=(PositiveWidth=0.1,NegativeWidth=1,PositiveHeight=0.1,NegativeHeight=1)
    ActiveRegionSize=0.1
    Debug=(bDrawDebug=false,DebugScale=25,DebugSegments=20,DebugThickness=0.25)
    DriverAxis=(X=-1,Y=0,Z=0)
    DriverItem=(Type=Bone,Name="upperarm_r")
    FalloffRegionScaleFactors=(PositiveWidth=0.8,NegativeWidth=1,PositiveHeight=0.8,NegativeHeight=1)
    FalloffSize=0.5
    FlipHeightScaling=true
    FlipWidthScaling=false
    OptionalParentItem=(Type=Bone,Name="spine_05")
    RotationOffset=(X=95,Y=-10,Z=170)
   )
   > RigVMVariableNode(Value=@n4.OutputParam) {variable=spr_upperarm_r_f_alpha}
   > n5 = &RigUnit_SphericalPoseReader(
    ActiveRegionScaleFactors=(PositiveWidth=2,NegativeWidth=2,PositiveHeight=2,NegativeHeight=2)
    ActiveRegionSize=0.1
    Debug=(bDrawDebug=false,DebugScale=25,DebugSegments=20,DebugThickness=0.25)
    DriverAxis=(X=1,Y=0,Z=0)
    DriverItem=(Type=Bone,Name="upperarm_l")
    FalloffRegionScaleFactors=(PositiveWidth=0.5,NegativeWidth=0.5,PositiveHeight=0.5,NegativeHeight=0.5)
    FalloffSize=0.5
    FlipHeightScaling=false
    FlipWidthScaling=false
    OptionalParentItem=(Type=Bone,Name="spine_05")
    RotationOffset=(X=180,Y=10,Z=0)
   )
   > RigVMVariableNode(Value=@n5.OutputParam) {variable=spr_upperarm_l_dn_alpha}
   > n6 = &RigUnit_SphericalPoseReader(
    ActiveRegionScaleFactors=(PositiveWidth=2,NegativeWidth=2,PositiveHeight=2,NegativeHeight=2)
    ActiveRegionSize=0.1
    Debug=(bDrawDebug=false,DebugScale=25,DebugSegments=20,DebugThickness=0.25)
    DriverAxis=(X=-1,Y=0,Z=0)
    DriverItem=(Type=Bone,Name="upperarm_r")
    FalloffRegionScaleFactors=(PositiveWidth=0.5,NegativeWidth=0.5,PositiveHeight=0.5,NegativeHeight=0.5)
    FalloffSize=0.5
    FlipHeightScaling=true
    FlipWidthScaling=false
    OptionalParentItem=(Type=Bone,Name="spine_05")
    RotationOffset=(X=0,Y=-15,Z=180)
   )
   > RigVMVariableNode(Value=@n6.OutputParam) {variable=spr_upperarm_r_dn_alpha}
   > /Script/ControlRig.RigUnit_SetCurveValue(Curve=spr_upperarm_l_dn_alpha, Value=RigVMVariableNode() {variable=spr_upperarm_l_dn_alpha}.Value) {is=struct RigVMUnitNode}
   &RigUnit_ParentConstraint(
    AdvancedSettings=(InterpolationType=Average,RotationOrderForFilter=XZY)
    Child=(Type=Bone,Name="dyn_coat_fr_MSTR_L")
    Filter=(TranslationFilter=(bX=true,bY=true,bZ=true),RotationFilter=(bX=true,bY=true,bZ=true),ScaleFilter=(bX=True,bY=True,bZ=True))
    Parents=((Item=(Type=Bone,Name="spine_04"),Weight=0.9),(Item=(Type=Bone,Name="clavicle_l"),Weight=0.1),(Item=(Type=Bone,Name="upperarm_twist_01_l"),Weight=0))
    Weight=1
    ExecutePin=@SequenceAggregate.B
   )
   > &RigUnit_ParentConstraint(
    AdvancedSettings=(InterpolationType=Average,RotationOrderForFilter=XZY)
    Child=(Type=Bone,Name="dyn_coat_MSTR_fL")
    Filter=(TranslationFilter=(bX=true,bY=true,bZ=true),RotationFilter=(bX=true,bY=true,bZ=true),ScaleFilter=(bX=True,bY=True,bZ=True))
    Parents=((Item=(Type=Bone,Name="spine_04"),Weight=0.8),(Item=(Type=Bone,Name="clavicle_l"),Weight=0.2),(Item=(Type=Bone,Name="upperarm_twist_01_l"),Weight=0))
    Weight=1
   )
   > &RigUnit_ParentConstraint(
    AdvancedSettings=(InterpolationType=Average,RotationOrderForFilter=XZY)
    Child=(Type=Bone,Name="dyn_coat_MSTR_L")
    Filter=(TranslationFilter=(bX=true,bY=true,bZ=true),RotationFilter=(bX=true,bY=true,bZ=true),ScaleFilter=(bX=True,bY=True,bZ=True))
    Parents=((Item=(Type=Bone,Name="spine_04"),Weight=0.7),(Item=(Type=Bone,Name="clavicle_l"),Weight=0.2),(Item=(Type=Bone,Name="upperarm_twist_01_l"),Weight=0.1))
    Weight=1
   )
   > &RigUnit_ParentConstraint(
    AdvancedSettings=(InterpolationType=Average,RotationOrderForFilter=XZY)
    Child=(Type=Bone,Name="dyn_coat_mid_MSTR_L")
    Filter=(TranslationFilter=(bX=true,bY=true,bZ=true),RotationFilter=(bX=true,bY=true,bZ=true),ScaleFilter=(bX=True,bY=True,bZ=True))
    Parents=((Item=(Type=Bone,Name="spine_04"),Weight=0.7),(Item=(Type=Bone,Name="clavicle_l"),Weight=0.3),(Item=(Type=Bone,Name="upperarm_twist_01_l"),Weight=0))
    Weight=1
   )
   n7 = RigVMVariableNode() {variable=spr_upperarm_l_f_alpha}
   > &RigUnit_ParentConstraint(
    AdvancedSettings=(InterpolationType=Average,RotationOrderForFilter=XZY)
    Child=(Type=Bone,Name="dyn_coat_fr_MSTR_L")
    Filter=(TranslationFilter=(bX=true,bY=true,bZ=true),RotationFilter=(bX=true,bY=true,bZ=true),ScaleFilter=(bX=True,bY=True,bZ=True))
    Parents=((Item=(Type=Bone,Name="spine_04"),Weight=0.6),(Item=(Type=Bone,Name="clavicle_l"),Weight=0.3),(Item=(Type=Bone,Name="upperarm_twist_01_l"),Weight=0.1))
    Weight=@n7.Value
   )
   > &RigUnit_ParentConstraint(
    AdvancedSettings=(InterpolationType=Average,RotationOrderForFilter=XZY)
    Child=(Type=Bone,Name="dyn_coat_MSTR_fL")
    Filter=(TranslationFilter=(bX=true,bY=true,bZ=true),RotationFilter=(bX=true,bY=true,bZ=true),ScaleFilter=(bX=True,bY=True,bZ=True))
    Parents=((Item=(Type=Bone,Name="spine_04"),Weight=0.6),(Item=(Type=Bone,Name="clavicle_l"),Weight=0.3),(Item=(Type=Bone,Name="upperarm_twist_01_l"),Weight=0.1))
    Weight=@n7.Value
   )
   > &RigUnit_ModifyTransforms(
    ItemToModify="((Item=(Type=Bone,Name=\"dyn_coat_MSTR_fL\"),Transform=(Rotation=(X=0,Y=0,Z=-0.0872,W=0.9962),Translation=(X=0,Y=5,Z=-3),Scale3D=(X=1,Y=1,Z=1))),(Item=(Type=Bone,Name=\"dyn_coat_fr_MSTR_L\"),Transform=(Rotation=(X=0,Y=0,Z=0,W=1),Translation=(X=0,Y=5,Z=3),Scale ...(truncated)"
    Weight=RigVMVariableNode() {variable=spr_upperarm_l_f_alpha}.Value
   )
   > &RigUnit_ModifyTransforms(
    ItemToModify="((Item=(Type=Bone,Name=\"dyn_coat_MSTR_fL\"),Transform=(Rotation=(X=0,Y=0,Z=0,W=1),Translation=(X=-3,Y=1,Z=-1),Scale3D=(X=1,Y=1,Z=1))),(Item=(Type=Bone,Name=\"dyn_coat_fr_MSTR_L\"),Transform=(Rotation=(X=0,Y=0,Z=0,W=1),Translation=(X=-3,Y=1,Z=1),Scale3D=(X=1,Y ...(truncated)"
    Weight=RigVMVariableNode() {variable=spr_upperarm_l_up_alpha}.Value
   )
   > &RigUnit_ModifyTransforms(
    ItemToModify="((Item=(Type=Bone,Name=\"dyn_coat_MSTR_fL\"),Transform=(Rotation=(X=0,Y=0,Z=0,W=1),Translation=(X=0,Y=0,Z=0),Scale3D=(X=1,Y=1,Z=1))),(Item=(Type=Bone,Name=\"dyn_coat_fr_MSTR_L\"),Transform=(Rotation=(X=0,Y=0,Z=0,W=1),Translation=(X=0,Y=0,Z=0),Scale3D=(X=1,Y=1, ...(truncated)"
    Weight=RigVMVariableNode() {variable=spr_upperarm_l_dn_alpha}.Value
   )
   &RigUnit_ParentConstraint(
    AdvancedSettings=(InterpolationType=Average,RotationOrderForFilter=XZY)
    Child=(Type=Bone,Name="dyn_coat_fr_MSTR_R")
    Filter=(TranslationFilter=(bX=true,bY=true,bZ=true),RotationFilter=(bX=true,bY=true,bZ=true),ScaleFilter=(bX=True,bY=True,bZ=True))
    Parents=((Item=(Type=Bone,Name="spine_04"),Weight=0.9),(Item=(Type=Bone,Name="clavicle_r"),Weight=0.1),(Item=(Type=Bone,Name="upperarm_twist_01_r"),Weight=0))
    Weight=1
    ExecutePin=@SequenceAggregate.C
   )
   > &RigUnit_ParentConstraint(
    AdvancedSettings=(InterpolationType=Average,RotationOrderForFilter=XZY)
    Child=(Type=Bone,Name="dyn_coat_MSTR_fR")
    Filter=(TranslationFilter=(bX=true,bY=true,bZ=true),RotationFilter=(bX=true,bY=true,bZ=true),ScaleFilter=(bX=True,bY=True,bZ=True))
    Parents=((Item=(Type=Bone,Name="spine_04"),Weight=0.8),(Item=(Type=Bone,Name="clavicle_r"),Weight=0.2),(Item=(Type=Bone,Name="upperarm_twist_01_r"),Weight=0))
    Weight=1
   )
   > &RigUnit_ParentConstraint(
    AdvancedSettings=(InterpolationType=Average,RotationOrderForFilter=XZY)
    Child=(Type=Bone,Name="dyn_coat_MSTR_R")
    Filter=(TranslationFilter=(bX=true,bY=true,bZ=true),RotationFilter=(bX=true,bY=true,bZ=true),ScaleFilter=(bX=True,bY=True,bZ=True))
    Parents=((Item=(Type=Bone,Name="spine_04"),Weight=0.6),(Item=(Type=Bone,Name="clavicle_r"),Weight=0.3),(Item=(Type=Bone,Name="upperarm_twist_01_r"),Weight=0.1))
    Weight=1
   )
   > &RigUnit_ParentConstraint(
    AdvancedSettings=(InterpolationType=Average,RotationOrderForFilter=XZY)
    Child=(Type=Bone,Name="dyn_coat_mid_MSTR_R")
    Filter=(TranslationFilter=(bX=true,bY=true,bZ=true),RotationFilter=(bX=true,bY=true,bZ=true),ScaleFilter=(bX=True,bY=True,bZ=True))
    Parents=((Item=(Type=Bone,Name="spine_04"),Weight=0.7),(Item=(Type=Bone,Name="clavicle_r"),Weight=0.3),(Item=(Type=Bone,Name="upperarm_twist_01_r"),Weight=0))
    Weight=1
   )
   n8 = RigVMVariableNode() {variable=spr_upperarm_r_f_alpha}
   > &RigUnit_ParentConstraint(
    AdvancedSettings=(InterpolationType=Average,RotationOrderForFilter=XZY)
    Child=(Type=Bone,Name="dyn_coat_fr_MSTR_R")
    Filter=(TranslationFilter=(bX=true,bY=true,bZ=true),RotationFilter=(bX=true,bY=true,bZ=true),ScaleFilter=(bX=True,bY=True,bZ=True))
    Parents=((Item=(Type=Bone,Name="spine_04"),Weight=0.6),(Item=(Type=Bone,Name="clavicle_r"),Weight=0.3),(Item=(Type=Bone,Name="upperarm_twist_01_r"),Weight=0.1))
    Weight=@n8.Value
   )
   > &RigUnit_ParentConstraint(
    AdvancedSettings=(InterpolationType=Average,RotationOrderForFilter=XZY)
    Child=(Type=Bone,Name="dyn_coat_MSTR_fR")
    Filter=(TranslationFilter=(bX=true,bY=true,bZ=true),RotationFilter=(bX=true,bY=true,bZ=true),ScaleFilter=(bX=True,bY=True,bZ=True))
    Parents=((Item=(Type=Bone,Name="spine_04"),Weight=0.6),(Item=(Type=Bone,Name="clavicle_r"),Weight=0.3),(Item=(Type=Bone,Name="upperarm_twist_01_r"),Weight=0.1))
    Weight=@n8.Value
   )
   > &RigUnit_ModifyTransforms(
    ItemToModify="((Item=(Type=Bone,Name=\"dyn_coat_MSTR_fR\"),Transform=(Rotation=(X=0,Y=0,Z=-0.0872,W=0.9962),Translation=(X=0,Y=-3,Z=-3),Scale3D=(X=1,Y=1,Z=1))),(Item=(Type=Bone,Name=\"dyn_coat_fr_MSTR_R\"),Transform=(Rotation=(X=0,Y=0,Z=0,W=1),Translation=(X=0,Y=-3,Z=3),Sca ...(truncated)"
    Weight=RigVMVariableNode() {variable=spr_upperarm_r_f_alpha}.Value
   )
   > &RigUnit_ModifyTransforms(
    ItemToModify="((Item=(Type=Bone,Name=\"dyn_coat_MSTR_fR\"),Transform=(Rotation=(X=0,Y=0,Z=0,W=1),Translation=(X=-3,Y=-1,Z=-1),Scale3D=(X=1,Y=1,Z=1))),(Item=(Type=Bone,Name=\"dyn_coat_fr_MSTR_R\"),Transform=(Rotation=(X=0,Y=0,Z=0,W=1),Translation=(X=-3,Y=-1,Z=1),Scale3D=(X=1 ...(truncated)"
    Weight=RigVMVariableNode() {variable=spr_upperarm_r_up_alpha}.Value
   )
   > &RigUnit_ModifyTransforms(
    ItemToModify="((Item=(Type=Bone,Name=\"dyn_coat_MSTR_fR\"),Transform=(Rotation=(X=0,Y=0,Z=0,W=1),Translation=(X=0,Y=0,Z=0),Scale3D=(X=1,Y=1,Z=1))),(Item=(Type=Bone,Name=\"dyn_coat_fr_MSTR_R\"),Transform=(Rotation=(X=0,Y=0,Z=0,W=1),Translation=(X=0,Y=0,Z=0),Scale3D=(X=1,Y=1, ...(truncated)"
    Weight=RigVMVariableNode() {variable=spr_upperarm_r_dn_alpha}.Value
   )
   Power = /Script/RigVM.RigVMFunction_MathDoublePow(B=2, A=RigVMVariableNode() {variable=spr_upperarm_l_dn_alpha}.Value) {is=struct RigVMUnitNode, template="Power::Execute(in A,in B,out Result)"}
   &RigUnit_ModifyTransforms(
    ItemToModify="((Item=(Type=Bone,Name=\"deform_armpit_2_l\"),Transform=(Rotation=(X=0,Y=0,Z=0,W=1),Translation=(X=0,Y=0,Z=3),Scale3D=(X=1,Y=1,Z=1))),(Item=(Type=Bone,Name=\"deform_armpit_1_l\"),Transform=(Rotation=(X=0,Y=0,Z=0,W=1),Translation=(X=0,Y=0,Z=8),Scale3D=(X=1,Y=1, ...(truncated)"
    WeightMinimum=0
    Weight=@Power.Result
    ExecutePin=@SequenceAggregate.D
   )
   > &RigUnit_ModifyTransforms(
    ItemToModify="((Item=(Type=Bone,Name=\"deform_armpit_2_r\"),Transform=(Rotation=(X=0,Y=0,Z=0,W=1),Translation=(X=0,Y=0,Z=-3),Scale3D=(X=1,Y=1,Z=1))),(Item=(Type=Bone,Name=\"deform_armpit_1_r\"),Transform=(Rotation=(X=0,Y=0,Z=0,W=1),Translation=(X=0,Y=0,Z=-8),Scale3D=(X=1,Y= ...(truncated)"
    Weight=&RigVMFunction_MathDoublePow(B=2, A=RigVMVariableNode() {variable=spr_upperarm_r_dn_alpha}.Value) {template="Power::Execute(in A,in B,out Result)"}.Result
   )
  ]}
]
hierarchy [
 root {loc=<0 0 0>, rot=<0 0 0 1>}
 pelvis {parent=root, loc=<0 -0.5301 98.0571>, rot=<0.0127 0.707 -0.0127 -0.707>}
 spine_01 {parent=pelvis, loc=<6.1208 -0.9328 0>, rot=<0 0 0.034 -0.9994>}
 spine_02 {parent=spine_01, loc=<6.5997 -0.2095 0>, rot=<0 0 0.0238 -0.9997>}
 spine_03 {parent=spine_02, loc=<6.4914 -0.1237 0>, rot=<0 0 0.0149 -0.9999>}
 spine_04 {parent=spine_03, loc=<6.9598 0.0794 0>, rot=<0 0 -0.0132 -0.9999>}
 spine_05 {parent=spine_04, loc=<6.9987 0 0>, rot=<0 0 -0.0015 -1>}
 neck_01 {parent=spine_05, loc=<11.3138 0.1529 0>, rot=<0 0 -0.2003 -0.9797>}
 neck_02 {parent=neck_01, loc=<4.5627 0 0>, rot=<0 0 0.0504 -0.9987>}
 head {parent=neck_02, loc=<5.3914 0 0>, rot=<0 0 0.115 -0.9934>}
 FACIAL_C_FacialRoot {parent=head, loc=<2.1524 -0.746 -0.0002>, rot=<0.4879 -0.5118 -0.4879 -0.5118>}
 FACIAL_C_UnderChin {parent=FACIAL_C_FacialRoot, loc=<0.0015 8.0522 4.3291>, rot=<0.4734 -0.0003 -0.0002 -0.8809>}
 FACIAL_L_UnderChin {parent=FACIAL_C_FacialRoot, loc=<2.3451 6.9234 2.9674>, rot=<0.3769 -0.3575 0.0714 -0.8515>}
 FACIAL_R_UnderChin {parent=FACIAL_C_FacialRoot, loc=<-2.349 6.9257 2.9602>, rot=<0.3769 0.3569 -0.0716 -0.8517>}
 FACIAL_C_Jaw {parent=FACIAL_C_FacialRoot, loc=<-0.0008 2.4782 1.9212>, rot=<0.3134 -0.0004 -0.0001 -0.9496>}
 FACIAL_C_Jawline {parent=FACIAL_C_Jaw, loc=<-0.0023 1.9183 7.0593>, rot=<0.2586 0 0 -0.966>}
 clavicle_l {parent=spine_05, loc=<5.4831 -0.1053 -3.2946>, rot=<0.0465 -0.7289 -0.009 -0.6829>}
 upperarm_l {parent=clavicle_l, loc=<9.8658 -1.2136 -0.4105>, rot=<0.0164 -0.3245 -0.0269 -0.9454>}
 lowerarm_l {parent=upperarm_l, loc=<26.465 -0.8345 0.259>, rot=<0.0036 0.0161 -0.2298 -0.9731>}
 hand_l {parent=lowerarm_l, loc=<22.7355 -0.3203 -0.6105>, rot=<-0.6784 -0.0223 0.0735 -0.7307>}
 thumb_01_l {parent=hand_l, loc=<1.3118 -0.3191 -2.459>, rot=<0.6546 -0.4042 0.0664 -0.6354>}
 thumb_02_l {parent=thumb_01_l, loc=<5.9336 -0.0448 0.0401>, rot=<-0.0037 0.0213 0.298 -0.9543>}
 thumb_03_l {parent=thumb_02_l, loc=<2.2223 0.3682 -0.0907>, rot=<0.0118 -0.013 0.0066 -0.9998>}
 index_metacarpal_l {parent=hand_l, loc=<1.6535 0.0183 -2.96>, rot=<0.0158 -0.0124 -0.0173 -0.9996>}
 index_01_l {parent=index_metacarpal_l, loc=<8.3008 0.0357 -0.0524>, rot=<-0.0044 0.0485 0.0924 -0.9945>}
 index_02_l {parent=index_01_l, loc=<3.3587 -0.3004 0.0452>, rot=<-0.002 0.0027 0.3225 -0.9466>}
 index_03_l {parent=index_02_l, loc=<2.5578 -0.123 -0.0252>, rot=<-0.002 0.0067 0.2558 -0.9667>}
 middle_metacarpal_l {parent=hand_l, loc=<2.1856 -0.1725 -0.5984>, rot=<0.006 0.0168 -0.0329 -0.9993>}
 middle_01_l {parent=middle_metacarpal_l, loc=<7.3275 -0.2447 0.0129>, rot=<-0.0208 0.0042 0.0858 -0.9961>}
 middle_02_l {parent=middle_01_l, loc=<4.0123 0.1401 0.1175>, rot=<-0.0218 0.0258 0.3512 -0.9357>}
 middle_03_l {parent=middle_02_l, loc=<2.6077 0.0534 0.1373>, rot=<0.0082 0.0188 0.1708 -0.9851>}
 ring_metacarpal_l {parent=hand_l, loc=<2.2074 -0.3625 1.3638>, rot=<-0.0284 0.0722 -0.0458 -0.9959>}
 ring_01_l {parent=ring_metacarpal_l, loc=<7.0295 -0.6197 -0.1446>, rot=<-0.0489 -0.0252 0.1892 -0.9804>}
 ring_02_l {parent=ring_01_l, loc=<3.6174 0.4647 -0.1027>, rot=<0 0 0.2806 -0.9598>}
 ring_03_l {parent=ring_02_l, loc=<2.4152 0.114 0.0199>, rot=<0 0 0.1189 -0.9929>}
 pinky_metacarpal_l {parent=hand_l, loc=<1.7616 -0.9753 3.3814>, rot=<-0.0397 0.0814 -0.069 -0.9935>}
 pinky_01_l {parent=pinky_metacarpal_l, loc=<7.1709 -0.4831 -0.1094>, rot=<-0.0654 -0.0035 0.2242 -0.9723>}
 pinky_02_l {parent=pinky_01_l, loc=<2.8274 0.0893 -0.1043>, rot=<0.0037 0.0011 0.2854 -0.9584>}
 pinky_03_l {parent=pinky_02_l, loc=<2.2046 0.0307 -0.0078>, rot=<-0.0002 0.0098 0.1445 -0.9895>}
 deform_wrist_dn_base_l {parent=hand_l, loc=<0.7676 -4.0212 0.8511>, rot=<-0.0126 -0.0008 0.0701 -0.9975>}
 deform_wrist_dn_l {parent=deform_wrist_dn_base_l, loc=<0 0 0>, rot=<0 0 0 1>}
 deform_wrist_up_base_l {parent=hand_l, loc=<-0.1235 2.9661 1.0272>, rot=<-0.0126 -0.0008 0.0701 -0.9975>}
 deform_wrist_up_l {parent=deform_wrist_up_base_l, loc=<0 0 0>, rot=<0 0 0 1>}
 lowerarm_twist_02_l {parent=lowerarm_l, loc=<10.6707 0 0>, rot=<0 0 0 1>}
 lowerarm_twist_01_l {parent=lowerarm_l, loc=<18.316 0 0>, rot=<0 0 0 1>}
 dyn_zipper_calf_L3 {parent=lowerarm_twist_01_l, loc=<2.1273 1.98 1.6621>, rot=<0.2773 -0.0124 -0.0073 -0.9607>}
 dyn_zipper_Hand_L {parent=lowerarm_twist_01_l, loc=<2.1273 1.98 1.6621>, rot=<0.2773 -0.0124 -0.0073 -0.9607>}
 deform_elbow_in_base_l {parent=lowerarm_l, loc=<1.9398 4.7088 -1.5913>, rot=<0.0285 -0.0335 0.0024 -0.999>}
 deform_elbow_in_l {parent=deform_elbow_in_base_l, loc=<0 0 0>, rot=<0 0 0 1>}
 deform_elbow_out_base_l {parent=lowerarm_l, loc=<-0.8233 -2.4892 -0.0666>, rot=<0.0285 -0.0335 0.0024 -0.999>}
 deform_elbow_out_l {parent=deform_elbow_out_base_l, loc=<0 0 0>, rot=<0 0 0 1>}
 upperarm_twist_01_l {parent=upperarm_l, loc=<7.2135 -0.1134 0>, rot=<0 0 0 1>}
 deform_upperarm_bk_base_l {parent=upperarm_twist_01_l, loc=<-5.9595 -4.3817 0.4059>, rot=<0 0.0007 0.0191 -0.9998>}
 deform_upperarm_bk_l {parent=deform_upperarm_bk_base_l, loc=<0 0 0>, rot=<0 0 0 1>}
 deform_upperarm_fr_base_l {parent=upperarm_twist_01_l, loc=<-5.7274 8.2053 -0.4141>, rot=<0 0.0007 0.0191 -0.9998>}
 deform_upperarm_fr_l {parent=deform_upperarm_fr_base_l, loc=<0 0 0>, rot=<0 0 0 1>}
 deform_upperarm_up_base_l {parent=upperarm_twist_01_l, loc=<-9.2216 0.1431 3.5601>, rot=<0 0.0007 0.0191 -0.9998>}
 deform_upperarm_up_l {parent=deform_upperarm_up_base_l, loc=<0 0 0>, rot=<0 0 0 1>}
 deform_armpit_1_base_l {parent=upperarm_twist_01_l, loc=<5.744 2.3786 -10.9542>, rot=<0 0 0 1>}
 deform_armpit_1_l {parent=deform_armpit_1_base_l, loc=<0 0 0>, rot=<0 0 0 1>}
 upperarm_twist_02_l {parent=upperarm_l, loc=<19.7998 -0.1134 0>, rot=<0 0 0 1>}
 deform_armpit_2_base_l {parent=upperarm_twist_02_l, loc=<0 0 -8.3776>, rot=<0 0 0 1>}
 deform_armpit_2_l {parent=deform_armpit_2_base_l, loc=<0 0 0>, rot=<0 0 0 1>}
 clavicle_r {parent=spine_05, loc=<5.4831 -0.1053 3.2946>, rot=<0.7289 0.0465 -0.6829 0.009>}
 upperarm_r {parent=clavicle_r, loc=<-9.8658 1.2136 0.4105>, rot=<0.0164 -0.3245 -0.0269 -0.9454>}
 lowerarm_r {parent=upperarm_r, loc=<-26.465 0.8345 -0.259>, rot=<0.0036 0.0161 -0.2298 -0.9731>}
 hand_r {parent=lowerarm_r, loc=<-22.7355 0.3203 0.6106>, rot=<-0.6784 -0.0223 0.0735 -0.7307>}
 thumb_01_r {parent=hand_r, loc=<-1.3118 0.3191 2.459>, rot=<0.6546 -0.4042 0.0664 -0.6354>}
 thumb_02_r {parent=thumb_01_r, loc=<-5.9336 0.0448 -0.0401>, rot=<-0.0037 0.0213 0.298 -0.9543>}
 thumb_03_r {parent=thumb_02_r, loc=<-2.2223 -0.3682 0.0907>, rot=<0.0118 -0.013 0.0066 -0.9998>}
 index_metacarpal_r {parent=hand_r, loc=<-1.6535 -0.0183 2.96>, rot=<0.0158 -0.0124 -0.0173 -0.9996>}
 index_01_r {parent=index_metacarpal_r, loc=<-8.3008 -0.0357 0.0524>, rot=<-0.0044 0.0485 0.0924 -0.9945>}
 index_02_r {parent=index_01_r, loc=<-3.3587 0.3004 -0.0452>, rot=<-0.002 0.0027 0.3225 -0.9466>}
 index_03_r {parent=index_02_r, loc=<-2.5578 0.123 0.0252>, rot=<-0.002 0.0067 0.2558 -0.9667>}
 middle_metacarpal_r {parent=hand_r, loc=<-2.1856 0.1725 0.5984>, rot=<0.006 0.0168 -0.0329 -0.9993>}
 middle_01_r {parent=middle_metacarpal_r, loc=<-7.3275 0.2447 -0.0129>, rot=<-0.0208 0.0042 0.0858 -0.9961>}
 middle_02_r {parent=middle_01_r, loc=<-4.0123 -0.1401 -0.1175>, rot=<-0.0218 0.0258 0.3512 -0.9357>}
 middle_03_r {parent=middle_02_r, loc=<-2.6077 -0.0534 -0.1373>, rot=<0.0082 0.0188 0.1708 -0.9851>}
 ring_metacarpal_r {parent=hand_r, loc=<-2.2074 0.3625 -1.3638>, rot=<-0.0284 0.0722 -0.0458 -0.9959>}
 ring_01_r {parent=ring_metacarpal_r, loc=<-7.0295 0.6197 0.1446>, rot=<-0.0489 -0.0252 0.1892 -0.9804>}
 ring_02_r {parent=ring_01_r, loc=<-3.6174 -0.4647 0.1027>, rot=<0 0 0.2806 -0.9598>}
 ring_03_r {parent=ring_02_r, loc=<-2.4152 -0.114 -0.0199>, rot=<0 0 0.1189 -0.9929>}
 pinky_metacarpal_r {parent=hand_r, loc=<-1.7616 0.9753 -3.3814>, rot=<-0.0397 0.0814 -0.069 -0.9935>}
 pinky_01_r {parent=pinky_metacarpal_r, loc=<-7.1709 0.4831 0.1094>, rot=<-0.0654 -0.0035 0.2242 -0.9723>}
 pinky_02_r {parent=pinky_01_r, loc=<-2.8274 -0.0893 0.1043>, rot=<0.0037 0.0011 0.2854 -0.9584>}
 pinky_03_r {parent=pinky_02_r, loc=<-2.2046 -0.0307 0.0078>, rot=<-0.0002 0.0098 0.1445 -0.9895>}
 deform_wrist_dn_base_r {parent=hand_r, loc=<-0.7676 4.0212 -0.8511>, rot=<-0.0126 -0.0008 0.0701 -0.9975>}
 deform_wrist_dn_r {parent=deform_wrist_dn_base_r, loc=<0 0 0>, rot=<0 0 0 1>}
 deform_wrist_up_base_r {parent=hand_r, loc=<0.1235 -2.9661 -1.0271>, rot=<-0.0126 -0.0008 0.0701 -0.9975>}
 deform_wrist_up_r {parent=deform_wrist_up_base_r, loc=<0 0 0>, rot=<0 0 0 1>}
 lowerarm_twist_02_r {parent=lowerarm_r, loc=<-10.6707 0 0>, rot=<0 0 0 -1>}
 lowerarm_twist_01_r {parent=lowerarm_r, loc=<-18.316 0 0>, rot=<0 0 0 -1>}
 dyn_zipper_calf_L4 {parent=lowerarm_twist_01_r, loc=<-2.2354 -2.681 -1.9389>, rot=<-0.0353 -0.9693 0.2431 -0.0062>}
 dyn_zipper_Hand_R {parent=lowerarm_twist_01_r, loc=<-2.2354 -2.681 -1.9389>, rot=<-0.0353 -0.9693 0.2431 -0.0062>}
 deform_elbow_in_base_r {parent=lowerarm_r, loc=<-1.9398 -4.7088 1.5913>, rot=<0.0285 -0.0335 0.0024 -0.999>}
 deform_elbow_in_r {parent=deform_elbow_in_base_r, loc=<-0.0001 0 0.0002>, rot=<0 0 0 1>}
 deform_elbow_out_base_r {parent=lowerarm_r, loc=<0.8233 2.4892 0.0666>, rot=<0.0285 -0.0335 0.0024 -0.999>}
 deform_elbow_out_r {parent=deform_elbow_out_base_r, loc=<0 0 0>, rot=<0 0 0 1>}
 upperarm_twist_01_r {parent=upperarm_r, loc=<-7.2135 0.1134 0>, rot=<0 0 0 -1>}
 deform_upperarm_bk_base_r {parent=upperarm_twist_01_r, loc=<5.9595 4.3817 -0.4059>, rot=<0 0.0007 0.0191 -0.9998>}
 deform_upperarm_bk_r {parent=deform_upperarm_bk_base_r, loc=<0 0 0>, rot=<0 0 0 1>}
 deform_upperarm_fr_base_r {parent=upperarm_twist_01_r, loc=<5.7274 -8.2053 0.4141>, rot=<0 0.0007 0.0191 -0.9998>}
 deform_upperarm_fr_r {parent=deform_upperarm_fr_base_r, loc=<0 0 0>, rot=<0 0 0 1>}
 deform_upperarm_up_base_r {parent=upperarm_twist_01_r, loc=<9.2216 -0.1431 -3.5601>, rot=<0 0.0007 0.0191 -0.9998>}
 deform_upperarm_up_r {parent=deform_upperarm_up_base_r, loc=<0 0 0>, rot=<0 0 0 -1>}
 deform_armpit_1_base_r {parent=upperarm_twist_01_r, loc=<-4.8611 -2.3786 10.406>, rot=<0 0 0 1>}
 deform_armpit_1_r {parent=deform_armpit_1_base_r, loc=<0 0 0>, rot=<0 0 0 1>}
 upperarm_twist_02_r {parent=upperarm_r, loc=<-19.7998 0.1134 0>, rot=<0 0 0 -1>}
 deform_armpit_2_base_r {parent=upperarm_twist_02_r, loc=<0.6803 0.1931 8.2362>, rot=<0 0 0 1>}
 deform_armpit_2_r {parent=deform_armpit_2_base_r, loc=<0 0 0>, rot=<0 0 0 1>}
 attach_backpack {parent=spine_05, loc=<0.0612 -6.4383 0>, rot=<-0.0093 -0.707 0.0093 -0.707>}
 Backpack_BR {parent=attach_backpack, type=Null, loc=<0 -0.6925 0>, rot=<0 0 0 1>}
 attach_cape {parent=spine_05, loc=<7.9381 -5.5951 0>, rot=<0.0616 -0.7044 -0.0616 -0.7044>}
 cape {parent=attach_cape, type=Null, loc=<0 0 -2.1587>, rot=<0 0 0 1>}
 dyn_zipper_spine_05 {parent=spine_05, loc=<2.8827 9.5899 5.9515>, rot=<0.2119 0.6375 0.7292 0.1306>}
 dyn_necklace_00 {parent=spine_05, loc=<10.2555 0.8371 -0.2016>, rot=<-0.3633 -0.6104 -0.6049 -0.36>}
 dyn_necklace_01 {parent=dyn_necklace_00, loc=<7.701 0.0331 -0.3798>, rot=<0 -0.3661 0 -0.9306>}
 dyn_collar_MSTR {parent=spine_05, loc=<7.89 7.4415 -4.422>, rot=<-0.3076 -0.6402 -0.635 -0.3038>}
 dyn_collar {parent=dyn_collar_MSTR, loc=<2.639 0 0>, rot=<0 0 0 1>}
 dyn_coat_fr_MSTR_R {parent=spine_04, loc=<4.2617 10.6029 6.7053>, rot=<-0.0429 0.5096 0.8578 -0.052>}
 dyn_coat_fr_01_R {parent=dyn_coat_fr_MSTR_R, loc=<1 0 0>, rot=<0 0 0 1>}
 dyn_coat_fr_02_R {parent=dyn_coat_fr_01_R, loc=<7.7003 0 0>, rot=<0 0 0 1>}
 dyn_zipper_coat_fr_02 {parent=dyn_coat_fr_02_R, loc=<5.0639 -1.9412 0.8083>, rot=<-0.0407 -0.0058 -0.0155 -0.999>}
 dyn_coat_bk_MSTR {parent=spine_04, loc=<5.6872 -8.612 0.0897>, rot=<-0.0138 -0.6984 0.7155 0.011>}
 dyn_coat_bk_01 {parent=dyn_coat_bk_MSTR, loc=<1 0 0>, rot=<0 0 0 1>}
 dyn_coat_bk_02 {parent=dyn_coat_bk_01, loc=<8.954 0 0>, rot=<0 0 0 1>}
 dyn_coat_mid_MSTR_R {parent=spine_04, loc=<5.106 -6.7812 10.9503>, rot=<-0.0357 -0.6488 0.7601 -0.0068>}
 dyn_coat_mid_01_R {parent=dyn_coat_mid_MSTR_R, loc=<1 0 0>, rot=<0 0 0 1>}
 dyn_coat_mid_02_R {parent=dyn_coat_mid_01_R, loc=<8.4283 0 0>, rot=<0 0 0 1>}
 dyn_coat_MSTR_fR {parent=spine_04, loc=<3.9087 5.5906 13.1268>, rot=<-0.0098 0.2075 0.9772 -0.0437>}
 dyn_coat_01_fR {parent=dyn_coat_MSTR_fR, loc=<1 0 0>, rot=<0 0 0 1>}
 dyn_coat_02_fR {parent=dyn_coat_01_fR, loc=<8.0406 0 0>, rot=<0 0 0 1>}
 dyn_coat_fr_MSTR_L {parent=spine_04, loc=<4.2617 10.6029 -6.7053>, rot=<0.0552 -0.8946 -0.4418 0.0388>}
 dyn_coat_fr_01_L {parent=dyn_coat_fr_MSTR_L, loc=<1 0 0>, rot=<0 0 0 1>}
 dyn_coat_fr_02_L {parent=dyn_coat_fr_01_L, loc=<7.7003 0 0>, rot=<0 0 0 1>}
 dyn_coat_mid_MSTR_L {parent=spine_04, loc=<5.106 -6.7812 -10.9503>, rot=<0.0046 -0.7982 0.6013 0.0361>}
 dyn_coat_mid_01_L {parent=dyn_coat_mid_MSTR_L, loc=<1.0001 0 0>, rot=<-0.0567 0 0 -0.9984>}
 dyn_coat_mid_02_L {parent=dyn_coat_mid_01_L, loc=<8.4277 0 0>, rot=<0 0 0 -1>}
 dyn_coat_MSTR_fL {parent=spine_04, loc=<3.9087 5.5906 -13.1268>, rot=<0.0437 -0.9764 -0.2112 0.0099>}
 dyn_coat_01_fL {parent=dyn_coat_MSTR_fL, loc=<1 0 0>, rot=<0 0 0 1>}
 dyn_coat_02_fL {parent=dyn_coat_01_fL, loc=<8.0399 0 0>, rot=<0 0 0 1>}
 dyn_coat_MSTR_L {parent=spine_04, loc=<4.4929 -0.9215 -14.4091>, rot=<0.0447 -0.9966 0.0698 -0.0027>}
 dyn_coat_01_L {parent=dyn_coat_MSTR_L, loc=<1 0 0>, rot=<0 0 0 1>}
 dyn_coat_02_L {parent=dyn_coat_01_L, loc=<8.0399 0 0>, rot=<0 0 0 1>}
 dyn_coat_MSTR_R {parent=spine_04, loc=<4.4929 -0.9215 14.4091>, rot=<0.0026 -0.0683 0.9967 -0.0447>}
 dyn_coat_01_R {parent=dyn_coat_MSTR_R, loc=<1 0 0>, rot=<0 0 0 1>}
 dyn_coat_02_R {parent=dyn_coat_01_R, loc=<8.0406 0 0>, rot=<0 0 0 1>}
 dyn_zipper_spine_01 {parent=spine_01, loc=<5.8198 7.8738 -0.0332>, rot=<0.0273 0.7066 0.7066 0.0273>}
 thigh_l {parent=pelvis, loc=<-1.0855 -0.2109 -8.7053>, rot=<0.0009 0.0355 0.0155 -0.9993>}
 calf_l {parent=thigh_l, loc=<-42.5284 -0.0019 -0.7297>, rot=<0.0006 -0.0161 -0.0391 -0.9991>}
 foot_l {parent=calf_l, loc=<-46.7401 -2.5014 -0.592>, rot=<0.0427 -0.0238 0.0353 -0.9982>}
 ball_l {parent=foot_l, loc=<-3.5658 10.8066 -0.1544>, rot=<0.0001 -0.0001 -0.7063 -0.7079>}
 calf_twist_02_l {parent=calf_l, loc=<-19.1546 -1.7943 -0.2638>, rot=<0.0237 0.0095 -0.0035 -0.9997>}
 dyn_zipper_calf_L {parent=calf_twist_02_l, loc=<2.3665 -0.8812 3.7238>, rot=<0.0399 0.1091 0.9921 0.0464>}
 calf_twist_01_l {parent=calf_l, loc=<-35.8483 -1.9197 -0.5794>, rot=<0.0237 0.0095 -0.0035 -0.9997>}
 deform_knee_in_base_l {parent=calf_l, loc=<0.421 -4.1223 0.458>, rot=<0.0017 0.0216 0.0001 -0.9998>}
 deform_knee_in_l {parent=deform_knee_in_base_l, loc=<0 0 0>, rot=<0 0 0 1>}
 deform_knee_out_base_l {parent=calf_l, loc=<-1.2545 3.3027 0.4099>, rot=<0.0017 0.0216 0.0001 -0.9998>}
 deform_knee_out_l {parent=deform_knee_out_base_l, loc=<0 0 0>, rot=<0 0 0 1>}
 Fix_foot_l {parent=calf_l, loc=<-38.1987 -0.5517 -0.9092>, rot=<0.0427 -0.0238 0.0353 -0.9982>}
 Fix_ball_l {parent=Fix_foot_l, loc=<-11.9489 8.2566 -0.4341>, rot=<0.0001 -0.0001 -0.7063 -0.7079>}
 thigh_twist_01_l {parent=thigh_l, loc=<-27.952 0 0>, rot=<0.0013 0 0 -1>}
 deform_groin_base_l {parent=thigh_l, loc=<-2.7829 5.4893 -0.1824>, rot=<0 0.0217 0 -0.9998>}
 deform_groin_l {parent=deform_groin_base_l, loc=<0 0 0>, rot=<0 0 0 1>}
 thigh_r {parent=pelvis, loc=<-1.0855 -0.2109 8.7053>, rot=<0.0355 -0.0009 0.9993 0.0155>}
 calf_r {parent=thigh_r, loc=<42.5284 0.0019 0.7297>, rot=<0.0006 -0.0161 -0.0391 -0.9991>}
 foot_r {parent=calf_r, loc=<46.7401 2.5014 0.592>, rot=<0.0427 -0.0238 0.0353 -0.9982>}
 ball_r {parent=foot_r, loc=<3.5658 -10.8066 0.1544>, rot=<0.0001 -0.0001 -0.7063 -0.7079>}
 calf_twist_02_r {parent=calf_r, loc=<19.1546 1.7943 0.2638>, rot=<0.0237 0.0095 -0.0035 -0.9997>}
 dyn_zipper_calf_L2 {parent=calf_twist_02_r, loc=<-2.3159 0.8818 -3.6884>, rot=<0.9947 -0.0543 -0.0317 0.0815>}
 calf_twist_01_r {parent=calf_r, loc=<35.8483 1.9197 0.5794>, rot=<0.0237 0.0095 -0.0035 -0.9997>}
 deform_knee_in_base_r {parent=calf_r, loc=<-0.421 4.1223 -0.458>, rot=<0.0017 0.0216 0.0001 -0.9998>}
 deform_knee_in_r {parent=deform_knee_in_base_r, loc=<0 0 0>, rot=<0 0 0 1>}
 deform_knee_out_base_r {parent=calf_r, loc=<1.2545 -3.3028 -0.4099>, rot=<0.0017 0.0216 0.0001 -0.9998>}
 deform_knee_out_r {parent=deform_knee_out_base_r, loc=<0 0 0>, rot=<0 0 0 1>}
 Fix_foot_r {parent=calf_r, loc=<38.1987 0.5517 0.9092>, rot=<0.0427 -0.0238 0.0353 -0.9982>}
 Fix_ball_r {parent=Fix_foot_r, loc=<11.9489 -8.2566 0.4341>, rot=<0.0001 -0.0001 -0.7063 -0.7079>}
 thigh_twist_01_r {parent=thigh_r, loc=<27.952 0 0>, rot=<0.0013 0 0 -1>}
 deform_groin_base_r {parent=thigh_r, loc=<2.7829 -5.4893 0.1824>, rot=<0 0.0217 0 -0.9998>}
 deform_groin_r {parent=deform_groin_base_r, loc=<0 0 0>, rot=<0 0 0 1>}
 deform_glute_base_l {parent=pelvis, loc=<-4.902 -9.0634 -7.4119>, rot=<0.0275 0.0501 -0.4472 -0.8926>}
 deform_glute_l {parent=deform_glute_base_l, loc=<-0.0131 -0.0327 0>, rot=<0 0 0 1>}
 deform_glute_base_r {parent=pelvis, loc=<-4.902 -9.0634 7.4119>, rot=<0.0501 -0.0275 0.8926 -0.4472>}
 deform_glute_r {parent=deform_glute_base_r, loc=<0.0131 0.0327 0>, rot=<0 0 0 1>}
 dyn_pendant_MSTR_L {parent=pelvis, loc=<9.8147 1.9575 -8.8766>, rot=<-0.0784 -0.9472 -0.218 -0.2215>}
 dyn_pendant_01_L {parent=dyn_pendant_MSTR_L, loc=<0.9756 0 0>, rot=<0 0.0078 0 -1>}
 dyn_pendant_02_L {parent=dyn_pendant_01_L, loc=<7.7249 0 0>, rot=<0 0 0 1>}
 dyn_pendant_03_L {parent=dyn_pendant_02_L, loc=<2.8763 0 0>, rot=<0 0 0 1>}
 dyn_pendant_04_L {parent=dyn_pendant_03_L, loc=<4.7273 0 0>, rot=<0 0 0 1>}
 dyn_pendant_03_a01_L {parent=dyn_pendant_03_L, loc=<0.832 0.7621 0.5195>, rot=<0 0 0 1>}
 dyn_pendant_03_a02_L {parent=dyn_pendant_03_a01_L, loc=<3.0347 0.1136 0>, rot=<0 0 0 1>}
 dyn_pendant_03_b01_L {parent=dyn_pendant_03_L, loc=<0.832 -0.019 0.5195>, rot=<0 0 0.0158 -0.9999>}
 dyn_pendant_03_b02_L {parent=dyn_pendant_03_b01_L, loc=<3.0347 0.1136 0>, rot=<0 0 0 1>}
 dyn_pendant_03_b03_L {parent=dyn_pendant_03_b02_L, loc=<3.1068 0 0>, rot=<0 0 0 1>}
 dyn_pendant_03_c_L {parent=dyn_pendant_03_L, loc=<0.0574 -0.1812 0.3419>, rot=<-0.3714 0.0059 0.0147 -0.9283>}
 dyn_zipper_leg_L {parent=pelvis, loc=<5.3488 6.1734 -6.0962>, rot=<-0.0531 -0.7293 -0.6804 -0.0495>}
 dyn_zipper_leg_R {parent=pelvis, loc=<5.344 6.1653 6.1148>, rot=<0.0491 0.6737 0.7354 0.0535>}
 gravity {parent=root, loc=<0 0 0>, rot=<0 0 0 1>}
 gravity_cloth {parent=root, loc=<0 0 0>, rot=<0 0 0 1>}
 spr_upperarm_l_dn_alpha {type=Curve, loc=<0 0 0>, rot=<0 0 0 1>}
]
variables [
 spr_upperarm_l_dn_alpha {type=double, default=0}
 spr_upperarm_l_f_alpha {type=double, default=0}
 spr_upperarm_l_up_alpha {type=double, default=0}
 spr_upperarm_r_dn_alpha {type=double, default=0}
 spr_upperarm_r_f_alpha {type=double, default=0}
 spr_upperarm_r_up_alpha {type=double, default=0}
]
