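# Physics asset for the F_MED_LoanFloat female character body, setting up capsule bodies for its dynamic coat and collar plus the spine and upper arms.
# Coat chains run MSTR→01→02 with tight swing and twist limits, their masters hanging from spine_04 on free-rotation constraints.
# Free-rotation cross constraints link neighbouring chains at both lower levels, from the back around each side to the front.
# A two-body limited collar chain attaches to spine_05.
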PhysicsAsset /BRCosmetics/Characters/Player/Female/Medium/Bodies/F_MED_LoanFloat/Meshes/F_MED_LoanFloat_Physics
bodies {
 dyn_coat_01_L [{shapeType=Capsule, length=2.5, radius=3.5, shapeName=dyn_coat_01_L_capsule, center=<3.1804 0.0071 0.1575>, rotation=<rot 0 0.0001 92.5648>}]
 dyn_coat_01_R [{shapeType=Capsule, length=2.5, radius=3.5, shapeName=dyn_coat_01_R_capsule, center=<3.1804 0.007 0.1575>, rotation=<rot 0 -0.0001 92.5648>}]
 dyn_coat_01_fL [{
   shapeType=Capsule
   length=4.2343
   radius=3
   shapeName=dyn_coat_01_fL_capsule
   center=<3.1804 0.0071 0.1575>
   rotation=<rot 0.0001 0.0001 92.5648>
  }]
 dyn_coat_01_fR [{
   shapeType=Capsule
   length=4.4683
   radius=3
   shapeName=dyn_coat_01_fR_capsule
   center=<3.1804 0.007 0.1575>
   rotation=<rot 0.0001 -0.0001 92.5648>
  }]
 dyn_coat_02_L [{
   shapeType=Capsule
   length=4.1383
   radius=3.5
   shapeName=dyn_coat_02_L_capsule
   center=<3.1804 0.0071 0.1575>
   rotation=<rot 0 0.0001 92.5648>
  }]
 dyn_coat_02_R [{
   shapeType=Capsule
   length=2.9556
   radius=3.5
   shapeName=dyn_coat_02_R_capsule
   center=<3.1804 0.007 0.1575>
   rotation=<rot 0 -0.0001 92.5648>
  }]
 dyn_coat_02_fL [{
   shapeType=Capsule
   length=3.3575
   radius=3
   shapeName=dyn_coat_02_fL_capsule
   center=<3.1804 0.0071 0.1575>
   rotation=<rot 0.0001 0.0001 92.5648>
  }]
 dyn_coat_02_fR [{shapeType=Capsule, length=3, radius=3, shapeName=dyn_coat_02_fR_capsule, center=<3.1804 0.007 0.1575>, rotation=<rot 0.0001 -0.0001 92.5648>}]
 dyn_coat_MSTR_L [{shapeType=Capsule, length=2.5, radius=3.5, shapeName=dyn_coat_MSTR_L_capsule, center=<3.1804 0.0071 0.1575>, rotation=<rot 0 0.0001 92.5648>}]
 dyn_coat_MSTR_R [{shapeType=Capsule, length=2.5, radius=3.5, shapeName=dyn_coat_MSTR_R_capsule, center=<3.1804 0.007 0.1575>, rotation=<rot 0 -0.0001 92.5648>}]
 dyn_coat_MSTR_fL [{
   shapeType=Capsule
   length=2.5
   radius=3.5
   shapeName=dyn_coat_MSTR_fL_capsule
   center=<3.1804 0.0071 0.1575>
   rotation=<rot 0.0001 0.0001 92.5648>
  }]
 dyn_coat_MSTR_fR [{
   shapeType=Capsule
   length=2.5
   radius=3.5
   shapeName=dyn_coat_MSTR_fR_capsule
   center=<3.1804 0.007 0.1575>
   rotation=<rot 0.0001 -0.0001 92.5648>
  }]
 dyn_coat_bk_01 [{
   shapeType=Capsule
   length=6.7228
   radius=3.5
   shapeName=dyn_coat_bk_01_capsule
   center=<3.1804 0.0276 0.6166>
   rotation=<rot -0.0001 0 92.5648>
  }]
 dyn_coat_bk_02 [{
   shapeType=Capsule
   length=5.4546
   radius=3.5
   shapeName=dyn_coat_bk_02_capsule
   center=<3.1804 0.0165 0.3674>
   rotation=<rot -0.0001 0 92.5648>
  }]
 dyn_coat_bk_MSTR [{
   shapeType=Capsule
   length=2.5
   radius=3.5
   shapeName=dyn_coat_bk_MSTR_capsule
   center=<3.1804 0.0071 0.1575>
   rotation=<rot -0.0001 0 92.5648>
  }]
 dyn_coat_fr_01_L [{
   shapeType=Capsule
   length=3.8421
   radius=3.75
   shapeName=dyn_coat_fr_01_L_capsule
   center=<4.0605 0.1058 2.3614>
   rotation=<rot 0.0001 0.0001 92.5648>
  }]
 dyn_coat_fr_01_R [{
   shapeType=Capsule
   length=4.2932
   radius=3.25
   shapeName=dyn_coat_fr_01_R_capsule
   center=<4.0605 0.0831 1.8548>
   rotation=<rot 0.0001 -0.0001 92.5648>
  }]
 dyn_coat_fr_02_L [{
   shapeType=Capsule
   length=3.9534
   radius=3
   shapeName=dyn_coat_fr_02_L_capsule
   center=<4.0605 0.0605 1.3513>
   rotation=<rot 0.0001 0.0001 92.5648>
  }]
 dyn_coat_fr_02_R [{
   shapeType=Capsule
   length=4.209
   radius=3.5
   shapeName=dyn_coat_fr_02_R_capsule
   center=<4.0605 0.0988 2.2049>
   rotation=<rot 0.0001 -0.0001 92.5648>
  }]
 dyn_coat_fr_MSTR_L [{
   shapeType=Capsule
   length=2.5
   radius=3.5
   shapeName=dyn_coat_fr_MSTR_L_capsule
   center=<3.1804 0.0071 0.1575>
   rotation=<rot 0.0001 0.0001 92.5648>
  }]
 dyn_coat_fr_MSTR_R [{
   shapeType=Capsule
   length=2.5
   radius=3.5
   shapeName=dyn_coat_fr_MSTR_R_capsule
   center=<3.1804 0.0071 0.1575>
   rotation=<rot 0.0001 -0.0001 92.5648>
  }]
 dyn_coat_mid_01_L [{
   shapeType=Capsule
   length=3.6306
   radius=3.5
   shapeName=dyn_coat_mid_01_L_capsule
   center=<3.1804 0.0071 0.1575>
   rotation=<rot -0.0001 0 92.5648>
  }]
 dyn_coat_mid_01_R [{
   shapeType=Capsule
   length=5.5184
   radius=3.5
   shapeName=dyn_coat_mid_01_R_capsule
   center=<3.1804 0.0148 0.3308>
   rotation=<rot -0.0001 0 92.5648>
  }]
 dyn_coat_mid_02_L [{
   shapeType=Capsule
   length=3.0429
   radius=3.5
   shapeName=dyn_coat_mid_02_L_capsule
   center=<3.1804 0.0071 0.1575>
   rotation=<rot -0.0001 0 92.5648>
  }]
 dyn_coat_mid_02_R [{
   shapeType=Capsule
   length=3.5102
   radius=3.5
   shapeName=dyn_coat_mid_02_R_capsule
   center=<3.1804 0.0071 0.1575>
   rotation=<rot -0.0001 0 92.5648>
  }]
 dyn_coat_mid_MSTR_L [{
   shapeType=Capsule
   length=2.5
   radius=3.5
   shapeName=dyn_coat_mid_MSTR_L_capsule
   center=<3.1804 0.0071 0.1575>
   rotation=<rot -0.0001 0 92.5648>
  }]
 dyn_coat_mid_MSTR_R [{
   shapeType=Capsule
   length=2.5
   radius=3.5
   shapeName=dyn_coat_mid_MSTR_R_capsule
   center=<3.1804 0.0071 0.1575>
   rotation=<rot -0.0001 0 92.5648>
  }]
 dyn_collar [{
   shapeType=Capsule
   length=1.7893
   radius=1.0617
   shapeName=dyn_collar_capsule
   center=<1.8907 0.1181 -0.2004>
   rotation=<rot 23.5056 -90 -18.8047>
  }]
 dyn_collar_MSTR [{
   shapeType=Capsule
   length=1.5241
   radius=0.7604
   shapeName=dyn_collar_MSTR_capsule
   center=<2.4157 0.36 -0.2534>
   rotation=<rot 1.8713 -90 -10.2875>
  }]
 spine_03 [{shapeType=Capsule, length=3.5, radius=6.5, shapeName=spine_03_capsule, center=<-2.1753 1.3029 0>}]
 spine_04 [{shapeType=Capsule, length=3.75, radius=7.75, shapeName=spine_03_capsule, center=<0.4264 2.1637 0>}]
 spine_05 [{shapeType=Capsule, length=7.5, radius=5.3656, center=<-1.409 3.1952 0>}]
 upperarm_l [{shapeType=Capsule, length=19.5586, radius=6.2442, center=<11.783 1.0221 2.9468>, rotation=<rot 84.7907 -179.9717 -179.9993>}]
 upperarm_r [{shapeType=Capsule, length=19.5586, radius=6.2442, center=<-10.4239 -1.4523 -1.5774>, rotation=<rot 84.7907 -179.9717 -179.9993>}]
}
constraints [
 {bone1Name=upperarm_l, bone2Name=spine_05, swing1LimitDegrees=45, swing2LimitDegrees=45, twistLimitDegrees=45}
 {bone1Name=upperarm_r, bone2Name=spine_05, swing1LimitDegrees=45, swing2LimitDegrees=45, twistLimitDegrees=45}
 {
  bone1Name=spine_05
  bone2Name=spine_04
  swing1LimitDegrees=45
  swing1Motion=Free
  swing2LimitDegrees=45
  swing2Motion=Free
  twistLimitDegrees=45
  twistMotion=Free
 }
 {bone1Name=dyn_coat_mid_01_L, bone2Name=dyn_coat_mid_MSTR_L, swing1LimitDegrees=5, swing2LimitDegrees=20, twistLimitDegrees=1}
 {bone1Name=dyn_coat_mid_02_L, bone2Name=dyn_coat_mid_01_L, swing1LimitDegrees=5, swing2LimitDegrees=20, twistLimitDegrees=1}
 {bone1Name=dyn_coat_mid_01_R, bone2Name=dyn_coat_mid_MSTR_R, swing1LimitDegrees=5, swing2LimitDegrees=20, twistLimitDegrees=1}
 {bone1Name=dyn_coat_mid_02_R, bone2Name=dyn_coat_mid_01_R, swing1LimitDegrees=5, swing2LimitDegrees=20, twistLimitDegrees=1}
 {bone1Name=dyn_coat_fr_02_R, bone2Name=dyn_coat_fr_01_R, swing1LimitDegrees=5, swing2LimitDegrees=20, twistLimitDegrees=1}
 {bone1Name=dyn_coat_fr_01_R, bone2Name=dyn_coat_fr_MSTR_R, swing1LimitDegrees=5, swing2LimitDegrees=20, twistLimitDegrees=1}
 {bone1Name=dyn_coat_fr_02_L, bone2Name=dyn_coat_fr_01_L, swing1LimitDegrees=5, swing2LimitDegrees=20, twistLimitDegrees=1}
 {bone1Name=dyn_coat_fr_01_L, bone2Name=dyn_coat_fr_MSTR_L, swing1LimitDegrees=5, swing2LimitDegrees=20, twistLimitDegrees=1}
 {bone1Name=dyn_coat_bk_02, bone2Name=dyn_coat_bk_01, swing1LimitDegrees=5, swing2LimitDegrees=20, twistLimitDegrees=1}
 {bone1Name=dyn_coat_bk_01, bone2Name=dyn_coat_bk_MSTR, swing1LimitDegrees=5, swing2LimitDegrees=20, twistLimitDegrees=1}
 {bone1Name=dyn_coat_02_fR, bone2Name=dyn_coat_01_fR, swing1LimitDegrees=5, swing2LimitDegrees=20, twistLimitDegrees=1}
 {bone1Name=dyn_coat_01_fR, bone2Name=dyn_coat_MSTR_fR, swing1LimitDegrees=5, swing2LimitDegrees=20, twistLimitDegrees=1}
 {bone1Name=dyn_coat_02_fL, bone2Name=dyn_coat_01_fL, swing1LimitDegrees=5, swing2LimitDegrees=20, twistLimitDegrees=1}
 {bone1Name=dyn_coat_01_fL, bone2Name=dyn_coat_MSTR_fL, swing1LimitDegrees=5, swing2LimitDegrees=20, twistLimitDegrees=1}
 {bone1Name=dyn_coat_02_R, bone2Name=dyn_coat_01_R, swing1LimitDegrees=1, swing2LimitDegrees=7, twistLimitDegrees=1}
 {bone1Name=dyn_coat_01_R, bone2Name=dyn_coat_MSTR_R, swing1LimitDegrees=1, swing2LimitDegrees=7, twistLimitDegrees=1}
 {bone1Name=dyn_coat_02_L, bone2Name=dyn_coat_01_L, swing1LimitDegrees=1, swing2LimitDegrees=7, twistLimitDegrees=1}
 {bone1Name=dyn_coat_01_L, bone2Name=dyn_coat_MSTR_L, swing1LimitDegrees=1, swing2LimitDegrees=7, twistLimitDegrees=1}
 {
  bone1Name=dyn_coat_MSTR_R
  bone2Name=spine_04
  swing1LimitDegrees=45
  swing1Motion=Free
  swing2LimitDegrees=45
  swing2Motion=Free
  twistLimitDegrees=45
  twistMotion=Free
 }
 {
  bone1Name=dyn_coat_MSTR_L
  bone2Name=spine_04
  swing1LimitDegrees=45
  swing1Motion=Free
  swing2LimitDegrees=45
  swing2Motion=Free
  twistLimitDegrees=45
  twistMotion=Free
 }
 {
  bone1Name=dyn_coat_MSTR_fL
  bone2Name=spine_04
  swing1LimitDegrees=45
  swing1Motion=Free
  swing2LimitDegrees=45
  swing2Motion=Free
  twistLimitDegrees=45
  twistMotion=Free
 }
 {
  bone1Name=dyn_coat_MSTR_fR
  bone2Name=spine_04
  swing1LimitDegrees=45
  swing1Motion=Free
  swing2LimitDegrees=45
  swing2Motion=Free
  twistLimitDegrees=45
  twistMotion=Free
 }
 {
  bone1Name=dyn_coat_bk_MSTR
  bone2Name=spine_04
  swing1LimitDegrees=45
  swing1Motion=Free
  swing2LimitDegrees=45
  swing2Motion=Free
  twistLimitDegrees=45
  twistMotion=Free
 }
 {
  bone1Name=dyn_coat_fr_MSTR_L
  bone2Name=spine_04
  swing1LimitDegrees=45
  swing1Motion=Free
  swing2LimitDegrees=45
  swing2Motion=Free
  twistLimitDegrees=45
  twistMotion=Free
 }
 {
  bone1Name=dyn_coat_fr_MSTR_R
  bone2Name=spine_04
  swing1LimitDegrees=45
  swing1Motion=Free
  swing2LimitDegrees=45
  swing2Motion=Free
  twistLimitDegrees=45
  twistMotion=Free
 }
 {
  bone1Name=dyn_coat_mid_MSTR_L
  bone2Name=spine_04
  swing1LimitDegrees=45
  swing1Motion=Free
  swing2LimitDegrees=45
  swing2Motion=Free
  twistLimitDegrees=45
  twistMotion=Free
 }
 {
  bone1Name=dyn_coat_mid_MSTR_R
  bone2Name=spine_04
  swing1LimitDegrees=45
  swing1Motion=Free
  swing2LimitDegrees=45
  swing2Motion=Free
  twistLimitDegrees=45
  twistMotion=Free
 }
 {
  bone1Name=dyn_coat_mid_01_L
  bone2Name=dyn_coat_bk_01
  swing1LimitDegrees=45
  swing1Motion=Free
  swing2LimitDegrees=45
  swing2Motion=Free
  twistLimitDegrees=45
  twistMotion=Free
 }
 {
  bone1Name=dyn_coat_mid_01_R
  bone2Name=dyn_coat_bk_01
  swing1LimitDegrees=45
  swing1Motion=Free
  swing2LimitDegrees=45
  swing2Motion=Free
  twistLimitDegrees=45
  twistMotion=Free
 }
 {
  bone1Name=dyn_coat_01_L
  bone2Name=dyn_coat_mid_01_L
  swing1LimitDegrees=45
  swing1Motion=Free
  swing2LimitDegrees=45
  swing2Motion=Free
  twistLimitDegrees=45
  twistMotion=Free
 }
 {
  bone1Name=dyn_coat_01_R
  bone2Name=dyn_coat_mid_01_R
  swing1LimitDegrees=45
  swing1Motion=Free
  swing2LimitDegrees=45
  swing2Motion=Free
  twistLimitDegrees=45
  twistMotion=Free
 }
 {
  bone1Name=dyn_coat_01_fL
  bone2Name=dyn_coat_01_L
  swing1LimitDegrees=45
  swing1Motion=Free
  swing2LimitDegrees=45
  swing2Motion=Free
  twistLimitDegrees=45
  twistMotion=Free
 }
 {
  bone1Name=dyn_coat_01_fR
  bone2Name=dyn_coat_01_R
  swing1LimitDegrees=45
  swing1Motion=Free
  swing2LimitDegrees=45
  swing2Motion=Free
  twistLimitDegrees=45
  twistMotion=Free
 }
 {
  bone1Name=dyn_coat_fr_01_L
  bone2Name=dyn_coat_01_fL
  swing1LimitDegrees=45
  swing1Motion=Free
  swing2LimitDegrees=45
  swing2Motion=Free
  twistLimitDegrees=45
  twistMotion=Free
 }
 {
  bone1Name=dyn_coat_fr_01_R
  bone2Name=dyn_coat_01_fR
  swing1LimitDegrees=45
  swing1Motion=Free
  swing2LimitDegrees=45
  swing2Motion=Free
  twistLimitDegrees=45
  twistMotion=Free
 }
 {
  bone1Name=dyn_coat_mid_02_L
  bone2Name=dyn_coat_bk_02
  swing1LimitDegrees=45
  swing1Motion=Free
  swing2LimitDegrees=45
  swing2Motion=Free
  twistLimitDegrees=45
  twistMotion=Free
 }
 {
  bone1Name=dyn_coat_mid_02_R
  bone2Name=dyn_coat_bk_02
  swing1LimitDegrees=45
  swing1Motion=Free
  swing2LimitDegrees=45
  swing2Motion=Free
  twistLimitDegrees=45
  twistMotion=Free
 }
 {
  bone1Name=dyn_coat_02_R
  bone2Name=dyn_coat_mid_02_R
  swing1LimitDegrees=45
  swing1Motion=Free
  swing2LimitDegrees=45
  swing2Motion=Free
  twistLimitDegrees=45
  twistMotion=Free
 }
 {
  bone1Name=dyn_coat_02_L
  bone2Name=dyn_coat_mid_02_L
  swing1LimitDegrees=45
  swing1Motion=Free
  swing2LimitDegrees=45
  swing2Motion=Free
  twistLimitDegrees=45
  twistMotion=Free
 }
 {
  bone1Name=dyn_coat_02_fL
  bone2Name=dyn_coat_02_L
  swing1LimitDegrees=45
  swing1Motion=Free
  swing2LimitDegrees=45
  swing2Motion=Free
  twistLimitDegrees=45
  twistMotion=Free
 }
 {
  bone1Name=dyn_coat_02_fR
  bone2Name=dyn_coat_02_R
  swing1LimitDegrees=45
  swing1Motion=Free
  swing2LimitDegrees=45
  swing2Motion=Free
  twistLimitDegrees=45
  twistMotion=Free
 }
 {
  bone1Name=dyn_coat_fr_02_L
  bone2Name=dyn_coat_02_fL
  swing1LimitDegrees=45
  swing1Motion=Free
  swing2LimitDegrees=45
  swing2Motion=Free
  twistLimitDegrees=45
  twistMotion=Free
 }
 {
  bone1Name=dyn_coat_fr_02_R
  bone2Name=dyn_coat_02_fR
  swing1LimitDegrees=45
  swing1Motion=Free
  swing2LimitDegrees=45
  swing2Motion=Free
  twistLimitDegrees=45
  twistMotion=Free
 }
 {
  bone1Name=spine_04
  bone2Name=spine_03
  swing1LimitDegrees=45
  swing1Motion=Free
  swing2LimitDegrees=45
  swing2Motion=Free
  twistLimitDegrees=45
  twistMotion=Free
 }
 {bone1Name=dyn_collar_MSTR, bone2Name=spine_05, swing1LimitDegrees=10, swing2LimitDegrees=2, twistLimitDegrees=2}
 {bone1Name=dyn_collar, bone2Name=dyn_collar_MSTR, swing1LimitDegrees=25, swing2LimitDegrees=15, twistLimitDegrees=10}
]
