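# Physics asset for the left-side pendant of the F_MED_LoanFloat female player body, set up for hip-hung dangle motion.
# Four equal capsules form a chain from a small master body constrained to the pelvis, with angular limits tighter at the first link than along the rest.
# Pelvis, left thigh and left forearm capsules provide nearby body shapes; their constraints, like the master's, leave rotation free.
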PhysicsAsset /BRCosmetics/Characters/Player/Female/Medium/Bodies/F_MED_LoanFloat/Meshes/F_MED_LoanFloat_pendant_Physics
bodies {
 dyn_pendant_01_L [{shapeType=Capsule, length=5, radius=2, shapeName=dyn_pendant_01_L_capsule, center=<2.5017 0 2.578>, rotation=<rot -89.9092 0.0413 -0.0411>}]
 dyn_pendant_02_L [{shapeType=Capsule, length=5, radius=2, shapeName=dyn_pendant_02_L_capsule, center=<2.5033 0 1.5992>, rotation=<rot -89.9092 0.0419 -0.0417>}]
 dyn_pendant_03_L [{shapeType=Capsule, length=5, radius=2, shapeName=dyn_pendant_03_L_capsule, center=<2.5041 0 1.07>, rotation=<rot -89.9092 0.0423 -0.0421>}]
 dyn_pendant_04_L [{shapeType=Capsule, length=5, radius=2, shapeName=dyn_pendant_04_L_capsule, center=<2.5043 0 0.9753>, rotation=<rot -89.9092 0.0421 -0.0419>}]
 dyn_pendant_MSTR_L [{
   shapeType=Capsule
   length=0.505
   radius=0.505
   shapeName=dyn_pendant_MSTR_L_capsule
   center=<0.1328 -0.0266 0.1374>
   rotation=<rot 0 102.1844 -84.6282>
  }]
 lowerarm_l [{
   shapeType=Capsule
   length=18.6941
   radius=4.6641
   shapeName=lowerarm_l_capsule
   center=<12.4165 1.1596 -1.7283>
   rotation=<rot -9.3961 -94.1507 88.2297>
  }]
 pelvis [{shapeType=Capsule, length=5.4168, radius=10.5921, shapeName=pelvis_capsule, center=<2.3897 -1.936 -0.2944>}]
 thigh_l [{
   shapeType=Capsule
   length=29.5771
   radius=7.4406
   shapeName=thigh_l_capsule
   center=<-18.3064 -0.3044 0.8519>
   rotation=<rot 0 89.627 -87.7208>
  }]
}
constraints [
 {bone1Name=dyn_pendant_01_L, bone2Name=dyn_pendant_MSTR_L, swing1LimitDegrees=40, swing2LimitDegrees=30, twistLimitDegrees=20}
 {bone1Name=dyn_pendant_02_L, bone2Name=dyn_pendant_01_L, swing1LimitDegrees=90, swing2LimitDegrees=45, twistLimitDegrees=40}
 {bone1Name=dyn_pendant_03_L, bone2Name=dyn_pendant_02_L, swing1LimitDegrees=90, swing2LimitDegrees=45, twistLimitDegrees=40}
 {bone1Name=dyn_pendant_04_L, bone2Name=dyn_pendant_03_L, swing1LimitDegrees=90, swing2LimitDegrees=45, twistLimitDegrees=40}
 {
  bone1Name=dyn_pendant_MSTR_L
  bone2Name=pelvis
  swing1LimitDegrees=45
  swing1Motion=Free
  swing2LimitDegrees=45
  swing2Motion=Free
  twistLimitDegrees=45
  twistMotion=Free
 }
 {
  bone1Name=thigh_l
  bone2Name=pelvis
  swing1LimitDegrees=45
  swing1Motion=Free
  swing2LimitDegrees=45
  swing2Motion=Free
  twistLimitDegrees=45
  twistMotion=Free
 }
 {
  bone1Name=lowerarm_l
  bone2Name=pelvis
  swing1LimitDegrees=45
  swing1Motion=Free
  swing2LimitDegrees=45
  swing2Motion=Free
  twistLimitDegrees=45
  twistMotion=Free
 }
]
