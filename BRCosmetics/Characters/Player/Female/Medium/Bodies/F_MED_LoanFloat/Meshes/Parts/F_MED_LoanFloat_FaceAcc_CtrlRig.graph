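# Control Rig for the F_MED_LoanFloat face accessory that poses a long back-hair bone chain against the body.
# Parent constraints move control nulls with the head, spine and pelvis progressively down the chain; null 09 is constrained twice, to spine_02 then pelvis.
# Construction caches the control list and pose; a Hermite Spline IK function without stretch takes those controls to drive the hair bones.
# A Modify Transforms offset node is left unwired.
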ControlRigBlueprint /BRCosmetics/Characters/Player/Female/Medium/Bodies/F_MED_LoanFloat/Meshes/Parts/F_MED_LoanFloat_FaceAcc_CtrlRig
graphs [
 ContainedGraph {nodes [
   n1 = RigVMFunctionEntryNode()
   SequenceAggregate = &RigVMFunction_Sequence(ExecuteContext=@n1.ExecuteContext)
   SequenceAggregate_1 = &RigVMFunction_Sequence(ExecuteContext=@SequenceAggregate.B)
   SequenceAggregate_2 = &RigVMFunction_Sequence(ExecuteContext=@SequenceAggregate_1.B)
   SequenceAggregate_3 = &RigVMFunction_Sequence(ExecuteContext=@SequenceAggregate_2.B)
   RigVMFunctionReturnNode(A=@SequenceAggregate.A, B=@SequenceAggregate_1.A, C=@SequenceAggregate_2.A, D=@SequenceAggregate_3.A, E=@SequenceAggregate_3.B)
  ]}
 ContainedGraph {nodes [
   n1 = RigVMFunctionEntryNode()
   SequenceAggregate = &RigVMFunction_Sequence(ExecuteContext=@n1.ExecuteContext)
   SequenceAggregate_1 = &RigVMFunction_Sequence(ExecuteContext=@SequenceAggregate.B)
   SequenceAggregate_2 = &RigVMFunction_Sequence(ExecuteContext=@SequenceAggregate_1.B)
   RigVMFunctionReturnNode(A=@SequenceAggregate.A, B=@SequenceAggregate_1.A, C=@SequenceAggregate_2.A, D=@SequenceAggregate_2.B)
  ]}
 RigVMFunctionLibrary {links [], nodes []}
 RigVMModel {nodes [
   RigUnit_BeginExecution = &RigUnit_BeginExecution()
   SequenceAggregate = RigVMAggregateNode(ExecuteContext=@RigUnit_BeginExecution.ExecutePin)
   &RigUnit_ParentConstraint(
    AdvancedSettings=(RotationOrderForFilter=XZY)
    Child=(Type=Null,Name="dyn_Hair_Back_01_Null")
    Filter=(TranslationFilter=(bX=true,bY=true,bZ=true),RotationFilter=(bX=true,bY=true,bZ=true),ScaleFilter=(bX=True,bY=True,bZ=True))
    Parents=((Item=(Type=Bone,Name="head"),Weight=1),(Item=(Type=Bone,Name="spine_05"),Weight=0))
    Weight=1
    ExecutePin=@SequenceAggregate.A
   )
   > &RigUnit_ParentConstraint(
    AdvancedSettings=(RotationOrderForFilter=XZY)
    Child=(Type=Null,Name="dyn_Hair_Back_02_Null")
    Filter=(TranslationFilter=(bX=true,bY=true,bZ=true),RotationFilter=(bX=true,bY=true,bZ=true),ScaleFilter=(bX=True,bY=True,bZ=True))
    Parents=((Item=(Type=Bone,Name="head"),Weight=1),(Item=(Type=Bone,Name="spine_05"),Weight=0))
    Weight=1
   )
   > &RigUnit_ParentConstraint(
    AdvancedSettings=(InterpolationType=Average,RotationOrderForFilter=XYZ)
    Child=(Type=Null,Name="dyn_Hair_Back_03_Null")
    Filter=(TranslationFilter=(bX=True,bY=True,bZ=True),RotationFilter=(bX=true,bY=true,bZ=true),ScaleFilter=(bX=True,bY=True,bZ=True))
    Parents=((Item=(Type=Bone,Name="head"),Weight=1),(Item=(Type=Bone,Name="spine_05"),Weight=0))
    Weight=1
   )
   > &RigUnit_ParentConstraint(
    AdvancedSettings=(RotationOrderForFilter=XZY)
    Child=(Type=Null,Name="dyn_Hair_Back_04_Null")
    Filter=(TranslationFilter=(bX=True,bY=True,bZ=True),RotationFilter=(bX=true,bY=true,bZ=true),ScaleFilter=(bX=True,bY=True,bZ=True))
    Parents=((Item=(Type=Bone,Name="head"),Weight=1),(Item=(Type=Bone,Name="spine_04"),Weight=0),(Item=(Type=Bone,Name="clavicle_l"),Weight=0))
    Weight=1
   )
   > &RigUnit_ParentConstraint(
    AdvancedSettings=(RotationOrderForFilter=XZY)
    Child=(Type=Null,Name="dyn_Hair_Back_05_null")
    Filter=(TranslationFilter=(bX=True,bY=True,bZ=True),RotationFilter=(bX=true,bY=true,bZ=true),ScaleFilter=(bX=True,bY=True,bZ=True))
    Parents=((Item=(Type=Bone,Name="head"),Weight=0.2),(Item=(Type=Bone,Name="spine_05"),Weight=0.8),(Item=(Type=Bone,Name="clavicle_r"),Weight=0))
    Weight=1
   )
   > &RigUnit_ParentConstraint(
    AdvancedSettings=(RotationOrderForFilter=XZY)
    Child=(Type=Null,Name="dyn_Hair_Back_06_Null")
    Filter=(TranslationFilter=(bX=True,bY=True,bZ=True),RotationFilter=(bX=true,bY=true,bZ=true),ScaleFilter=(bX=True,bY=True,bZ=True))
    Parents=((Item=(Type=Bone,Name="None"),Weight=0),(Item=(Type=Bone,Name="spine_04"),Weight=1))
    Weight=1
   )
   > &RigUnit_ParentConstraint(
    AdvancedSettings=(RotationOrderForFilter=XZY)
    Child=(Type=Null,Name="dyn_Hair_Back_07_Null")
    Filter=(TranslationFilter=(bX=True,bY=True,bZ=True),RotationFilter=(bX=true,bY=true,bZ=true),ScaleFilter=(bX=True,bY=True,bZ=True))
    Parents=((Item=(Type=Bone,Name="None"),Weight=0),(Item=(Type=Bone,Name="spine_03"),Weight=1))
    Weight=1
   )
   > &RigUnit_ParentConstraint(
    AdvancedSettings=(RotationOrderForFilter=XZY)
    Child=(Type=Null,Name="dyn_Hair_Back_08_Null")
    Filter=(TranslationFilter=(bX=True,bY=True,bZ=True),RotationFilter=(bX=true,bY=true,bZ=true),ScaleFilter=(bX=True,bY=True,bZ=True))
    Parents=((Item=(Type=Bone,Name="None"),Weight=0),(Item=(Type=Bone,Name="spine_03"),Weight=1))
    Weight=1
   )
   > &RigUnit_ParentConstraint(
    AdvancedSettings=(RotationOrderForFilter=XZY)
    Child=(Type=Null,Name="dyn_Hair_Back_09_null")
    Filter=(TranslationFilter=(bX=True,bY=True,bZ=True),RotationFilter=(bX=true,bY=true,bZ=true),ScaleFilter=(bX=True,bY=True,bZ=True))
    Parents=((Item=(Type=Bone,Name="None"),Weight=0),(Item=(Type=Bone,Name="spine_02"),Weight=1))
    Weight=1
   )
   > &RigUnit_ParentConstraint(
    AdvancedSettings=(RotationOrderForFilter=XZY)
    Child=(Type=Null,Name="dyn_Hair_Back_10_null")
    Filter=(TranslationFilter=(bX=True,bY=True,bZ=True),RotationFilter=(bX=true,bY=true,bZ=true),ScaleFilter=(bX=True,bY=True,bZ=True))
    Parents=((Item=(Type=Bone,Name="None"),Weight=0),(Item=(Type=Bone,Name="spine_01"),Weight=1))
    Weight=1
   )
   > &RigUnit_ParentConstraint(
    AdvancedSettings=(RotationOrderForFilter=XZY)
    Child=(Type=Null,Name="dyn_Hair_Back_09_null")
    Filter=(TranslationFilter=(bX=True,bY=True,bZ=True),RotationFilter=(bX=true,bY=true,bZ=true),ScaleFilter=(bX=True,bY=True,bZ=True))
    Parents=((Item=(Type=Bone,Name="None"),Weight=0),(Item=(Type=Bone,Name="pelvis"),Weight=1))
    Weight=1
   )
   > &RigUnit_ParentConstraint(
    AdvancedSettings=(RotationOrderForFilter=XZY)
    Child=(Type=Null,Name="dyn_Hair_Back_11_null")
    Filter=(TranslationFilter=(bX=True,bY=True,bZ=True),RotationFilter=(bX=true,bY=true,bZ=true),ScaleFilter=(bX=True,bY=True,bZ=True))
    Parents=((Item=(Type=Bone,Name="None"),Weight=0),(Item=(Type=Bone,Name="pelvis"),Weight=1))
    Weight=1
   )
   > &RigUnit_ParentConstraint(
    AdvancedSettings=(RotationOrderForFilter=XZY)
    Child=(Type=Null,Name="dyn_Hair_Back_12_null")
    Filter=(TranslationFilter=(bX=True,bY=True,bZ=True),RotationFilter=(bX=true,bY=true,bZ=true),ScaleFilter=(bX=True,bY=True,bZ=True))
    Parents=((Item=(Type=Bone,Name="None"),Weight=0),(Item=(Type=Bone,Name="pelvis"),Weight=1))
    Weight=1
   )
   > &RigUnit_ParentConstraint(
    AdvancedSettings=(RotationOrderForFilter=XZY)
    Child=(Type=Null,Name="dyn_Hair_Back_13_null")
    Filter=(TranslationFilter=(bX=True,bY=True,bZ=True),RotationFilter=(bX=true,bY=true,bZ=true),ScaleFilter=(bX=True,bY=True,bZ=True))
    Parents=((Item=(Type=Bone,Name="None"),Weight=0),(Item=(Type=Bone,Name="pelvis"),Weight=1))
    Weight=1
   )
   > &RigUnit_ParentConstraint(
    AdvancedSettings=(RotationOrderForFilter=XZY)
    Child=(Type=Null,Name="dyn_Hair_Back_14_null")
    Filter=(TranslationFilter=(bX=True,bY=True,bZ=True),RotationFilter=(bX=true,bY=true,bZ=true),ScaleFilter=(bX=True,bY=True,bZ=True))
    Parents=((Item=(Type=Bone,Name="None"),Weight=0),(Item=(Type=Bone,Name="pelvis"),Weight=1))
    Weight=1
   )
   > &RigUnit_ParentConstraint(
    AdvancedSettings=(RotationOrderForFilter=XZY)
    Child=(Type=Null,Name="dyn_Hair_Back_15_null")
    Filter=(TranslationFilter=(bX=True,bY=True,bZ=True),RotationFilter=(bX=true,bY=true,bZ=true),ScaleFilter=(bX=True,bY=True,bZ=True))
    Parents=((Item=(Type=Bone,Name="None"),Weight=0),(Item=(Type=Bone,Name="pelvis"),Weight=1))
    Weight=1
   )
   > &RigUnit_ParentConstraint(
    AdvancedSettings=(RotationOrderForFilter=XZY)
    Child=(Type=Null,Name="dyn_Hair_Back_16_null")
    Filter=(TranslationFilter=(bX=True,bY=True,bZ=True),RotationFilter=(bX=true,bY=true,bZ=true),ScaleFilter=(bX=True,bY=True,bZ=True))
    Parents=((Item=(Type=Bone,Name="None"),Weight=0),(Item=(Type=Bone,Name="pelvis"),Weight=1))
    Weight=1
   )
   RerouteNode_2 = RigVMDispatchNode(Values="(((),Type=Bone,Name=\"dyn_Hair_Back_01\"),((),Type=Bone,Name=\"dyn_Hair_Back_02\"),((),Type=Bone,Name=\"dyn_Hair_Back_03\"),((),Type=Bone,Name=\"dyn_Hair_Back_04\"),((),Type=Bone,Name=\"dyn_Hair_Back_05\"),((),Type=Bone,Name=\"dyn_Hair_Back_06\"),((),Type=Bone,Name=\"d ...(truncated)") {template="DISPATCH_RigVMDispatch_ArrayMake(in Values,out Array)"}
   > SplineIK = RigVMFunctionReferenceNode(
    Primary Axis=(X=0,Y=-0.1,Z=0)
    Samples Per Segment=10
    Secondary Spline Direction=(X=0,Y=0,Z=0)
    Spline Mode=Hermite
    Stretch=false
    Up Axis=(X=0,Y=0,Z=0)
    debug=true
    Bones=@RerouteNode_2.Array
    Controls=RigVMVariableNode() {variable=dyn_Hair_Back_ctrls}.Value
   )
   &RigUnit_ModifyTransforms(
    ExecutePin=()
    ItemToModify="((Item=(Type=Null,Name=\"dyn_Hair_Back_05_Null\"),Transform=(Rotation=(X=0,Y=0,Z=0,W=1),Translation=(X=2,Y=0,Z=0),Scale3D=(X=1,Y=1,Z=1))),(Item=(Type=Null,Name=\"dyn_Hair_Back_07_Null\"),Transform=(Rotation=(X=0,Y=0,Z=0,W=1),Translation=(X=-3,Y=0,Z=0),Scale3D= ...(truncated)"
    Weight=1
   )
  ]}
 RigVMModel Construction Event Graph {nodes [
   n1 = &RigUnit_PrepareForExecution()
   SequenceAggregate = RigVMAggregateNode(ExecuteContext=@n1.ExecutePin)
   RerouteNode_6_1 = RigVMDispatchNode(Values="(((),Type=Control,Name=\"dyn_Hair_Back_01_ctrl\"),((),Type=Control,Name=\"dyn_Hair_Back_02_ctrl\"),((),Type=Control,Name=\"dyn_Hair_Back_03_ctrl\"),((),Type=Control,Name=\"dyn_Hair_Back_04_ctrl\"),((),Type=Control,Name=\"dyn_Hair_Back_05_ctrl\"),((),Type=Control,Nam ...(truncated)") {template="DISPATCH_RigVMDispatch_ArrayMake(in Values,out Array)"}
   RigVMVariableNode(Value=@RerouteNode_6_1.Array, ExecuteContext=@SequenceAggregate.A) {variable=dyn_Hair_Back_ctrls}
   > RigVMVariableNode(Value=/Script/ControlRig.RigUnit_HierarchyGetPoseItemArray(ElementType=All, Initial=false, ItemsToGet=@RerouteNode_6_1.Array) {is=struct RigVMUnitNode}.Pose) {variable=dyn_Hair_Back_pose}
  ]}
]
hierarchy [
 root {loc=<0 0 0>, rot=<0 0 0 1>}
 pelvis {parent=root, loc=<0 -0.5301 98.0571>, rot=<0.0127 0.707 -0.0127 -0.707>}
 spine_01 {parent=pelvis, loc=<6.1208 -0.9328 0>, rot=<0 0 0.034 -0.9994>}
 spine_02 {parent=spine_01, loc=<6.5996 -0.2095 0>, rot=<0 0 0.0238 -0.9997>}
 spine_03 {parent=spine_02, loc=<6.4914 -0.1237 0>, rot=<0 0 0.0149 -0.9999>}
 spine_04 {parent=spine_03, loc=<6.9598 0.0794 0>, rot=<0 0 -0.0132 -0.9999>}
 spine_05 {parent=spine_04, loc=<6.9987 0 0>, rot=<0 0 -0.0015 -1>}
 neck_01 {parent=spine_05, loc=<11.3138 0.1529 0>, rot=<0 0 -0.2003 -0.9797>}
 neck_02 {parent=neck_01, loc=<4.5627 0 0>, rot=<0 0 0.0504 -0.9987>}
 head {parent=neck_02, loc=<5.3914 0 0>, rot=<0 0 0.115 -0.9934>}
 dyn_Hair_Back_01 {parent=head, loc=<10.5153 -3.4657 -0.3258>, rot=<0 0 -0.004 -1>}
 dyn_Hair_Back_02 {parent=dyn_Hair_Back_01, loc=<1.5992 -6.634 0.3438>, rot=<0 0 0.5593 -0.8289>}
 dyn_Hair_Back_03 {parent=dyn_Hair_Back_02, loc=<-0.8129 -6.3517 0>, rot=<0 0 0.1706 -0.9853>}
 dyn_Hair_Back_04 {parent=dyn_Hair_Back_03, loc=<0 -9.1536 -0.3554>, rot=<0 0 0 1>}
 dyn_Hair_Back_05 {parent=dyn_Hair_Back_04, loc=<0 -8.2382 0>, rot=<0 0 0 1>}
 dyn_Hair_Back_06 {parent=dyn_Hair_Back_05, loc=<-0.0095 -9.0225 0>, rot=<0 0 0.0154 -0.9999>}
 dyn_Hair_Back_07 {parent=dyn_Hair_Back_06, loc=<0 -7.6497 0>, rot=<0 0 0 1>}
 dyn_Hair_Back_08 {parent=dyn_Hair_Back_07, loc=<0 -9.4969 0>, rot=<0 0 0 1>}
 dyn_Hair_Back_09 {parent=dyn_Hair_Back_08, loc=<0 -8.1037 0>, rot=<0 0 0 1>}
 dyn_Hair_Back_10 {parent=dyn_Hair_Back_09, loc=<0 -9.2812 0>, rot=<0 0 0 1>}
 dyn_Hair_Back_11 {parent=dyn_Hair_Back_10, loc=<0 -10.2138 0>, rot=<0 0 0 1>}
 dyn_Hair_Back_12 {parent=dyn_Hair_Back_11, loc=<0 -10.5269 0>, rot=<0 0 0 1>}
 dyn_Hair_Back_13 {parent=dyn_Hair_Back_12, loc=<0 -10.6477 0>, rot=<0 0 0 1>}
 dyn_Hair_Back_14 {parent=dyn_Hair_Back_13, loc=<0 -10.194 0>, rot=<0 0 0 1>}
 dyn_Hair_Back_15 {parent=dyn_Hair_Back_14, loc=<0 -11.4238 0>, rot=<0 0 0 1>}
 dyn_Hair_Back_16 {parent=dyn_Hair_Back_15, loc=<0 -6.9887 0>, rot=<0 0 0 1>}
 dyn_Hair_Back_15a {parent=dyn_Hair_Back_15, loc=<2.0901 -0.4799 0.3846>, rot=<0.5859 -0.4602 0.4121 -0.5245>}
 dyn_Hair_Back_15b {parent=dyn_Hair_Back_15, loc=<2.2259 -5.2986 0.4284>, rot=<0.4903 -0.4887 0.4836 -0.5357>}
 dyn_Hair_Back_15b1 {parent=dyn_Hair_Back_15b, loc=<6.892 0 -0.3033>, rot=<0 -0.0246 0 -0.9997>}
 clavicle_l {parent=spine_05, loc=<5.4831 -0.1053 -3.2946>, rot=<0.0465 -0.7289 -0.009 -0.6829>}
 upperarm_l {parent=clavicle_l, loc=<9.8658 -1.2136 -0.4105>, rot=<0.0164 -0.3245 -0.0269 -0.9454>}
 lowerarm_l {parent=upperarm_l, loc=<26.465 -0.8345 0.259>, rot=<0.0036 0.0161 -0.2298 -0.9731>}
 clavicle_r {parent=spine_05, loc=<5.4831 -0.1053 3.2946>, rot=<0.7289 0.0465 -0.6829 0.009>}
 upperarm_r {parent=clavicle_r, loc=<-9.8658 1.2136 0.4105>, rot=<0.0164 -0.3245 -0.0269 -0.9454>}
 upperarm_twist_01_r {parent=upperarm_r, loc=<-7.2135 0.1134 0>, rot=<0 0 0 -1>}
 lowerarm_r {parent=upperarm_r, loc=<-26.465 0.8345 -0.259>, rot=<0.0036 0.0161 -0.2298 -0.9731>}
 dyn_coat_bk_MSTR {parent=spine_04, loc=<5.6872 -8.612 0.0897>, rot=<-0.0138 -0.6984 0.7155 0.011>}
 dyn_coat_bk_01 {parent=dyn_coat_bk_MSTR, loc=<1 0 0>, rot=<0 0 0 1>}
 dyn_coat_bk_02 {parent=dyn_coat_bk_01, loc=<8.954 0 0>, rot=<0 0 0 1>}
 dyn_coat_mid_MSTR_R {parent=spine_04, loc=<5.106 -6.7812 10.9503>, rot=<-0.0357 -0.6488 0.7601 -0.0068>}
 dyn_coat_mid_01_R {parent=dyn_coat_mid_MSTR_R, loc=<1 0 0>, rot=<0 0 0 1>}
 dyn_coat_mid_02_R {parent=dyn_coat_mid_01_R, loc=<8.4283 0 0>, rot=<0 0 0 1>}
 dyn_coat_mid_MSTR_L {parent=spine_04, loc=<5.106 -6.7812 -10.9503>, rot=<0.0046 -0.7982 0.6013 0.0361>}
 dyn_coat_mid_01_L {parent=dyn_coat_mid_MSTR_L, loc=<1.0001 0 0>, rot=<-0.0567 0 0 -0.9984>}
 dyn_coat_mid_02_L {parent=dyn_coat_mid_01_L, loc=<8.4277 0 0>, rot=<0 0 0 -1>}
 thigh_l {parent=pelvis, loc=<-1.0855 -0.2109 -8.7053>, rot=<0.0009 0.0355 0.0155 -0.9993>}
 calf_l {parent=thigh_l, loc=<-42.5284 -0.0019 -0.7297>, rot=<0.0006 -0.0161 -0.0391 -0.9991>}
 foot_l {parent=calf_l, loc=<-46.7401 -2.5014 -0.592>, rot=<0.0427 -0.0238 0.0353 -0.9982>}
 calf_twist_02_l {parent=calf_l, loc=<-19.1546 -1.7943 -0.2638>, rot=<0.0237 0.0095 -0.0035 -0.9997>}
 calf_twist_01_l {parent=calf_l, loc=<-35.8483 -1.9197 -0.5794>, rot=<0.0237 0.0095 -0.0035 -0.9997>}
 thigh_twist_01_l {parent=thigh_l, loc=<-27.952 0 0>, rot=<0.0013 0 0 -1>}
 thigh_r {parent=pelvis, loc=<-1.0855 -0.2109 8.7053>, rot=<0.0355 -0.0009 0.9993 0.0155>}
 calf_r {parent=thigh_r, loc=<42.5284 0.0019 0.7297>, rot=<0.0006 -0.0161 -0.0391 -0.9991>}
 foot_r {parent=calf_r, loc=<46.7401 2.5014 0.592>, rot=<0.0427 -0.0238 0.0353 -0.9982>}
 calf_twist_02_r {parent=calf_r, loc=<19.1546 1.7943 0.2638>, rot=<0.0237 0.0095 -0.0035 -0.9997>}
 calf_twist_01_r {parent=calf_r, loc=<35.8483 1.9197 0.5794>, rot=<0.0237 0.0095 -0.0035 -0.9997>}
 thigh_twist_01_r {parent=thigh_r, loc=<27.952 0 0>, rot=<0.0013 0 0 -1>}
 dyn_Hair_Back_01_null {parent=root, type=Null, loc=<0.3258 -5.2398 162.5206>, rot=<0.0001 -0.7071 -0.0001 0.7071>}
 dyn_Hair_Back_01_ctrl {controlType=EULER_TRANSFORM, parent=dyn_Hair_Back_01_null, type=Control, loc=<0 0 0>, rot=<0 0 0 1>}
 dyn_Hair_Back_02_null {parent=root, type=Null, loc=<-0.018 -11.8742 164.118>, rot=<-0.3956 0.5861 0.3956 -0.5861>}
 dyn_Hair_Back_02_ctrl {controlType=EULER_TRANSFORM, parent=dyn_Hair_Back_02_null, type=Control, loc=<0 0 0>, rot=<0 0 0 1>}
 dyn_Hair_Back_03_null {parent=root, type=Null, loc=<-0.018 -13.4961 157.9232>, rot=<0.4898 -0.51 -0.4898 0.51>}
 dyn_Hair_Back_03_ctrl {controlType=EULER_TRANSFORM, parent=dyn_Hair_Back_03_null, type=Control, loc=<0 0 0>, rot=<0 0 0 1>}
 dyn_Hair_Back_04_null {parent=root, type=Null, loc=<0.3374 -13.8673 148.7772>, rot=<0.4898 -0.51 -0.4898 0.51>}
 dyn_Hair_Back_04_ctrl {controlType=EULER_TRANSFORM, parent=dyn_Hair_Back_04_null, type=Control, loc=<0 0 0>, rot=<0 0 0 1>}
 dyn_Hair_Back_05_null {parent=root, type=Null, loc=<0.3374 -14.2013 140.5458>, rot=<0.4898 -0.51 -0.4898 0.51>}
 dyn_Hair_Back_05_ctrl {controlType=EULER_TRANSFORM, parent=dyn_Hair_Back_05_null, type=Control, loc=<0 0 0>, rot=<0 0 0 1>}
 dyn_Hair_Back_06_null {parent=root, type=Null, loc=<0.3374 -14.5577 131.5303>, rot=<-0.4975 0.5025 0.4975 -0.5025>}
 dyn_Hair_Back_06_ctrl {controlType=EULER_TRANSFORM, parent=dyn_Hair_Back_06_null, type=Control, loc=<0 0 0>, rot=<0 0 0 1>}
 dyn_Hair_Back_07_null {parent=root, type=Null, loc=<0.3374 -14.6328 123.8811>, rot=<-0.4975 0.5025 0.4975 -0.5025>}
 dyn_Hair_Back_07_ctrl {controlType=EULER_TRANSFORM, parent=dyn_Hair_Back_07_null, type=Control, loc=<0 0 0>, rot=<0 0 0 1>}
 dyn_Hair_Back_08_null {parent=root, type=Null, loc=<0.3374 -14.7262 114.3846>, rot=<-0.4975 0.5025 0.4975 -0.5025>}
 dyn_Hair_Back_08_ctrl {controlType=EULER_TRANSFORM, parent=dyn_Hair_Back_08_null, type=Control, loc=<0 0 0>, rot=<0 0 0 1>}
 dyn_Hair_Back_09_null {parent=root, type=Null, loc=<0.3374 -14.8058 106.2813>, rot=<0.4975 -0.5025 -0.4975 0.5025>}
 dyn_Hair_Back_09_ctrl {controlType=EULER_TRANSFORM, parent=dyn_Hair_Back_09_null, type=Control, loc=<0 0 0>, rot=<0 0 0 1>}
 dyn_Hair_Back_10_null {parent=root, type=Null, loc=<0.3374 -14.897 97.0006>, rot=<-0.4975 0.5025 0.4975 -0.5025>}
 dyn_Hair_Back_10_ctrl {controlType=EULER_TRANSFORM, parent=dyn_Hair_Back_10_null, type=Control, loc=<0 0 0>, rot=<0 0 0 1>}
 dyn_Hair_Back_11_null {parent=root, type=Null, loc=<0.3374 -14.9974 86.7873>, rot=<0.4975 -0.5025 -0.4975 0.5025>}
 dyn_Hair_Back_11_ctrl {controlType=EULER_TRANSFORM, parent=dyn_Hair_Back_11_null, type=Control, loc=<0 0 0>, rot=<0 0 0 1>}
 dyn_Hair_Back_12_null {parent=root, type=Null, loc=<0.3374 -15.1008 76.2609>, rot=<-0.4975 0.5025 0.4975 -0.5025>}
 dyn_Hair_Back_12_ctrl {controlType=EULER_TRANSFORM, parent=dyn_Hair_Back_12_null, type=Control, loc=<0 0 0>, rot=<0 0 0 1>}
 dyn_Hair_Back_13_null {parent=root, type=Null, loc=<0.3374 -15.2055 65.6137>, rot=<0.4975 -0.5025 -0.4975 0.5025>}
 dyn_Hair_Back_13_ctrl {controlType=EULER_TRANSFORM, parent=dyn_Hair_Back_13_null, type=Control, loc=<0 0 0>, rot=<0 0 0 1>}
 dyn_Hair_Back_14_null {parent=root, type=Null, loc=<0.3374 -15.3056 55.4202>, rot=<-0.4975 0.5025 0.4975 -0.5025>}
 dyn_Hair_Back_14_ctrl {controlType=EULER_TRANSFORM, parent=dyn_Hair_Back_14_null, type=Control, loc=<0 0 0>, rot=<0 0 0 1>}
 dyn_Hair_Back_15_null {parent=root, type=Null, loc=<0.3374 -15.4179 43.9969>, rot=<0.4975 -0.5025 -0.4975 0.5025>}
 dyn_Hair_Back_15_ctrl {controlType=EULER_TRANSFORM, parent=dyn_Hair_Back_15_null, type=Control, loc=<0 0 0>, rot=<0 0 0 1>}
 dyn_Hair_Back_16_null {parent=root, type=Null, loc=<0.3374 -15.4866 37.0085>, rot=<-0.4975 0.5025 0.4975 -0.5025>}
 dyn_Hair_Back_16_ctrl {controlType=EULER_TRANSFORM, parent=dyn_Hair_Back_16_null, type=Control, loc=<0 0 0>, rot=<0 0 0 1>}
 gravity {parent=root, loc=<0 0 0>, rot=<0 0 0 1>}
]
variables [
 dyn_Hair_BackL_ctrls {
  type="TArray<FRigElementKey>"
  default=((Type=Control,Name="dyn_Hair_BackL_02"),(Type=Control,Name="dyn_Hair_BackL_03"),(Type=Control,Name="dyn_Hair_BackL_04"),(Type=Control,Name="dyn_Hair_BackL_05"),(Type=Control,Name="dyn_Hair_BackL_06"))
 }
 dyn_Hair_BackL_pose {type=FRigPose, default=(Elements=,HierarchyTopologyVersion=-1,PoseHash=-1)}
 dyn_Hair_BackR_ctrls {
  type="TArray<FRigElementKey>"
  default=((Type=Control,Name="dfrm_arm_spline_01_r_ctrl"),(Type=Control,Name="dfrm_arm_spline_03_r_ctrl"),(Type=Control,Name="dfrm_arm_spline_05_r_ctrl"),(Type=Control,Name="dfrm_arm_spline_07_r_ctrl"),(Type=Control,Name="dfrm_arm_spline_09_r_ctrl"))
 }
 dyn_Hair_BackR_pose {type=FRigPose, default=(Elements=,HierarchyTopologyVersion=-1,PoseHash=-1)}
 dyn_Hair_Back_ctrls {
  type="TArray<FRigElementKey>"
  default=((Type=Control,Name="dyn_Hair_BackL_02"),(Type=Control,Name="dyn_Hair_BackL_03"),(Type=Control,Name="dyn_Hair_BackL_04"),(Type=Control,Name="dyn_Hair_BackL_05"),(Type=Control,Name="dyn_Hair_BackL_06"))
 }
 dyn_Hair_Back_pose {type=FRigPose, default=(Elements=,HierarchyTopologyVersion=-1,PoseHash=-1)}
]
